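# Additive UI material for the BarnTowel fuel meter, drawing a tinted frame and radial fuel fill from texture channels.
# Two radial-progress masks drive Fuel and DelayFuel, with their difference forming a separately colored delay segment.
# UVs support haze distortion from two panning noise samples, center scaling, and a switchable horizontal cosine shake.
# Clamped CameraVelocity supplies a drift offset, so the meter shifts with camera motion.
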Material /BarnTowelGameplay/UI/V2/M_UI_Barntowel_Fuel
graph {
 properties {bAutomaticallySetUsageInEditor=false, blendMode=BLEND_Additive, materialDomain=MD_UI, usedShadingModels=Default Lit}
 nodes [
  n1 = TextureSample(UVs=NamedRerouteUsage()) {samplerSource=SSM_Clamp_WorldGroupSettings, samplerType=SAMPLERTYPE_LinearColor, texture=&/T_UI_BarnTowel_FuelMeter}
  n2 = ScalarParameter() {defaultValue=1, parameterName=FuelOpacity}
  n3 = Add(A=NamedRerouteUsage(), B=Constant2Vector() {r=5.75})
  n4 = Constant() {r=93.05}
  n5 = Constant() {r=-6.12}
  n6 = Multiply(
   A=@n1.G
   B=MaterialFunctionCall!(UVs=@n3, Progress=ScalarParameter() {defaultValue=0.25, parameterName=Fuel}, StartValue=@n4, Range=@n5) {materialFunction=/FortUILibrary/MaterialFunctions/Basics/MF_UI_RadialProgress}.Result
  )
  n7 = Multiply(A=@n6, B=VectorParameter() {parameterName=Color, defaultValue=<rgba 0.8715 0 0.0092 1>}.RGB)
  n8 = Multiply(
   A=@n1.G
   B=MaterialFunctionCall(UVs=@n3, Progress=ScalarParameter() {defaultValue=0.5, parameterName=DelayFuel}, StartValue=@n4, Range=@n5) {materialFunction=&MF_UI_RadialProgress}.Result
  )
  n9 = Multiply(A=Subtract(A=@n8, B=@n6), B=VectorParameter() {parameterName=Delay Color, sortPriority=1, defaultValue=<rgba 0.9931 0.83 0 1>}.RGB)
  n10 = ScalarParameter() {defaultValue=1, parameterName=DelayFuelOpacity}
  Add(
   A=Multiply(A=VectorParameter() {parameterName=Frame Color, defaultValue=<rgba 0.4078 0.4824 1 1>}.RGB, B=@n1.R)
   B=Add(A=Multiply(A=@n2, B=@n7), B=Multiply(A=@n9, B=@n10))
  )
  Add(A=Multiply(A=@n1.R) {constB=2}, B=Multiply(A=Add(A=@n6, B=@n8)))
  n11 = TextureCoordinate()
  n12 = TextureSample(UVs=Panner(Coordinate=@n11) {speedX=0.1, speedY=0.02}) {texture=/Game/Packages/Fortress_Transylvania/Haunted_Forest/Swamps/Textures/2ndInfectedMapLinear2}
  n13 = TextureSample(UVs=Panner(Coordinate=Multiply(A=@n11) {constB=0.76}) {speedX=-0.05, speedY=-0.01}) {texture=&2ndInfectedMapLinear2}
  n14 = ConstantBiasScale(None=ComponentMask(None=Multiply(A=@n12.RGB, B=@n13.RGB)) {g=true, r=true}) {bias=-0.125, scale=2}
  n15 = ScalarParameter() {parameterName=Haze, sliderMax=0.2}
  NamedRerouteDeclaration(None=Add(A=TextureCoordinate(), B=Multiply(A=@n14, B=@n15) {constB=0.2})) {name=GlitchedUVs, nodeColor=<rgba 0 0 0 1>}
  n16 = MaterialFunctionCall(UVs=NamedRerouteUsage(), Texture Scale=ScalarParameter() {defaultValue=0.7, parameterName=UVScalar}) {materialFunction=&ScaleUVsByCenter}
  n17 = NamedRerouteUsage()
  n18 = Time()
  n19 = Add(A=Cosine(None=Multiply(A=@n18) {constB=8}))
  n20 = AppendVector(A=LinearInterpolate(Alpha=Multiply(A=@n19) {constB=0.5}) {constA=-0.125, constAlpha=1, constB=0.125}, B=Constant())
  n21 = ScalarParameter() {parameterName=ShakeSwitch}
  NamedRerouteDeclaration(None=Add(A=Add(A=@n16.UVs, B=@n17), B=LinearInterpolate(B=@n20, Alpha=@n21)) {constB=0}) {name=MasterUV, nodeColor=<rgba 0 0 0 1>}
  n22 = ComponentMask(None=VectorParameter() {parameterName=CameraVelocity}.RGB) {g=true, r=true}
  n23 = Constant2Vector() {g=0.6, r=7}
  n24 = Constant2Vector() {desc=Max Drift, g=0.37, r=0.23}
  NamedRerouteDeclaration(None=Clamp(None=Divide(A=@n22, B=@n23), Min=Multiply(A=@n24) {constB=-1}, Max=@n24)) {name=CameraOffset, nodeColor=<rgba 0 0 0 1>}
 ]
}
