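# Sound cue that plays a fuel warning for the BarnTowel gameplay item, classed as a Fortnite gadget sound.
# A mixer layers the fuel warning wave, slightly boosted, with a looping countdown wave that starts after a fixed half-second delay.
# The cue plays at reduced overall volume and borrows the jetpack attenuation settings for spatial falloff.
# The countdown loop has no end inside the cue, so it plays until stopped externally.
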SoundCue /BarnTowelGameplay/Assets/Sounds/_various/BarnTowel_Fuel_Warning_Cue
root=Mixer_0
nodes [
 {id=Delay_0, children [WavePlayer_1], properties {delayMax=0.5, delayMin=0.5}}
 {id=Mixer_0, children [WavePlayer_0,Delay_0], properties {inputVolume [1.1,1]}}
 {id=WavePlayer_0, properties {soundWaveAssetPtr=&/BarnTowel_Fuel_Warning}}
 {id=WavePlayer_1, properties {bLooping=true, soundWaveAssetPtr=&/BarnTowel_Fuel_CountDown}}
]
properties {
 attenuationSettings=/Game/Sounds/Attenuation/Fort_Jetpack_Att
 maxDistance=13067
 soundClassObject=&Fort_Gadgets
 volumeMultiplier=0.5
}
