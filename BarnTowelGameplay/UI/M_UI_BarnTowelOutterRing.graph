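# Additive UI material that draws the BarnTowel outer ring as a tinted, rotatable circle texture.
# The circle is masked by a sharpened reflected gradient and multiplied by a Color parameter.
# Two counter-panning noise samples, scaled by Haze, build glitch-distorted UVs.
# A CameraVelocity parameter becomes a clamped drift offset, with both exposed as named reroutes for the ring's UVs and gradient.
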Material /BarnTowelGameplay/UI/M_UI_BarnTowelOutterRing
graph {
 properties {bAutomaticallySetUsageInEditor=false, blendMode=BLEND_Additive, materialDomain=MD_UI, usedShadingModels=Default Lit}
 nodes [
  n1 = Add(
   A=MaterialFunctionCall(UVs=NamedRerouteUsage(), "Rotation Angle (0-1)"=ScalarParameter() {parameterName=Rotation}) {materialFunction=&CustomRotator}.Rotated Values
   B=NamedRerouteUsage()
  )
  n2 = MaterialFunctionCall!(UVs, Offset=ComponentMask(None=NamedRerouteUsage()) {g=true}, RotationCenter, RotationAngle=Constant() {r=0.25}) {materialFunction=/FortUILibrary/MaterialFunctions/Gradients/MF_UI_ReflectedGradient}
  Multiply(
   A=Multiply(
    A=TextureSample(UVs=@n1) {samplerType=SAMPLERTYPE_LinearGrayscale, texture=&/T_UI_BarnTowel_OuterCircle}.R
    B=Power(Base=@n2.Result) {constExponent=5}
   )
   B=VectorParameter() {parameterName=Color, defaultValue=<rgba 0.4062 0.4841 1 1>}.RGB
  )
  n3 = TextureCoordinate()
  n4 = TextureSample(UVs=Panner(Coordinate=@n3) {speedX=0.1, speedY=0.02}) {texture=/Game/Packages/Fortress_Transylvania/Haunted_Forest/Swamps/Textures/2ndInfectedMapLinear2}
  n5 = TextureSample(UVs=Panner(Coordinate=Multiply(A=@n3) {constB=0.76}) {speedX=-0.05, speedY=-0.01}) {texture=&2ndInfectedMapLinear2}
  n6 = ConstantBiasScale(None=ComponentMask(None=Multiply(A=@n4.RGB, B=@n5.RGB)) {g=true, r=true}) {bias=-0.125, scale=2}
  n7 = ScalarParameter() {parameterName=Haze, sliderMax=0.2}
  NamedRerouteDeclaration(None=Add(A=TextureCoordinate(), B=Multiply(A=@n6, B=@n7) {constB=0.2})) {name=GlitchedUVs, nodeColor=<rgba 0 0 0 1>}
  n8 = ComponentMask(None=VectorParameter() {parameterName=CameraVelocity}.RGB) {g=true, r=true}
  n9 = Constant2Vector() {g=0.84, r=9.6}
  n10 = Constant2Vector() {desc=Max Drift, g=0.02, r=0.02}
  NamedRerouteDeclaration(None=Clamp(None=Divide(A=@n8, B=@n9), Min=Multiply(A=@n10) {constB=-1}, Max=@n10)) {name=CamVel, nodeColor=<rgba 0 0 0 1>}
 ]
}
