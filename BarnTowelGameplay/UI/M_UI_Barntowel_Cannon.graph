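# Additive UI material that draws the Barntowel cannon reticle, with cooldown and ready states.
# Base, accent and timer outlines come from hexagon and triangle SDFs combined with Max, forming chamfered shapes.
# A radially mirrored box mask cuts the accent, which ColorFlash tints from white to red, while CooldownProgress drives a hard-edged radial mask on the timer.
# The Ready? parameter blends between two layer combinations, scaled by ReticleOpacity.
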Material /BarnTowelGameplay/UI/M_UI_Barntowel_Cannon
graph {
 expressions {
  MaterialExpressionAdd_12 {
   class=MaterialExpressionAdd
   inputs [
    A {from=MaterialExpressionScalarParameter_18}
    B {from=""}
   ]
   properties {constB=0.4333}
  }
  MaterialExpressionAdd_13 {
   class=MaterialExpressionAdd
   inputs [
    A {from=MaterialExpressionScalarParameter_19}
    B {from=""}
   ]
   properties {constB=0.4333}
  }
  MaterialExpressionAdd_2 {
   class=MaterialExpressionAdd
   inputs [
    A {from=MaterialExpressionScalarParameter_8}
    B {from=""}
   ]
   properties {constB=0.4333}
  }
  MaterialExpressionAppendVector_0 {
   class=MaterialExpressionAppendVector
   inputs [
    A {from=MaterialExpressionMaterialFunctionCall_2, output=Inner}
    B {from=MaterialExpressionMaterialFunctionCall_3, output=Outline}
   ]
  }
  MaterialExpressionAppendVector_1 {
   class=MaterialExpressionAppendVector
   inputs [
    A {from=MaterialExpressionMax_0, output=""}
    B {from=MaterialExpressionMaterialFunctionCall_3, output=SDFOutline}
   ]
  }
  MaterialExpressionAppendVector_2 {
   class=MaterialExpressionAppendVector
   inputs [
    A {from=MaterialExpressionMaterialFunctionCall_7, output=Inner}
    B {from=MaterialExpressionMaterialFunctionCall_8, output=Outline}
   ]
  }
  MaterialExpressionAppendVector_3 {
   class=MaterialExpressionAppendVector
   inputs [
    A {from=MaterialExpressionMax_1, output=""}
    B {from=MaterialExpressionMaterialFunctionCall_8, output=SDFOutline}
   ]
  }
  MaterialExpressionAppendVector_4 {
   class=MaterialExpressionAppendVector
   inputs [
    A {from=MaterialExpressionMaterialFunctionCall_11, output=Inner}
    B {from=MaterialExpressionMaterialFunctionCall_12, output=Outline}
   ]
  }
  MaterialExpressionCeil_0 {class=MaterialExpressionCeil, inputs [None {from=MaterialExpressionSaturate_0, output=""}], properties {}}
  MaterialExpressionComponentMask_0 {class=MaterialExpressionComponentMask, inputs [None {from=MaterialExpressionStaticSwitch_0}], properties {r=true}}
  MaterialExpressionComponentMask_1 {class=MaterialExpressionComponentMask, inputs [None {from=MaterialExpressionStaticSwitch_0}], properties {g=true}}
  MaterialExpressionComponentMask_2 {class=MaterialExpressionComponentMask, inputs [None {from=MaterialExpressionStaticSwitch_2}], properties {r=true}}
  MaterialExpressionComponentMask_3 {class=MaterialExpressionComponentMask, inputs [None {from=MaterialExpressionStaticSwitch_2}], properties {g=true}}
  MaterialExpressionComponentMask_5 {class=MaterialExpressionComponentMask, inputs [None {from=MaterialExpressionAppendVector_4}], properties {g=true}}
  MaterialExpressionComponentMask_7 {class=MaterialExpressionComponentMask, inputs [None {from=MaterialExpressionMaterialFunctionCall_15, output=UVs}], properties {r=true}}
  MaterialExpressionComposite_0 {class=MaterialExpressionComposite, inputs [], properties {subgraphName=HexUV/Sides}}
  MaterialExpressionComposite_1 {class=MaterialExpressionComposite, inputs [], properties {subgraphName=TriangleUV/Radius/Sides}}
  MaterialExpressionComposite_4 {class=MaterialExpressionComposite, inputs [], properties {subgraphName=HexUV/Sides}}
  MaterialExpressionComposite_5 {class=MaterialExpressionComposite, inputs [], properties {subgraphName=TriangleUV/Radius/Sides}}
  MaterialExpressionComposite_6 {class=MaterialExpressionComposite, inputs [], properties {subgraphName=HexUV/Sides}}
  MaterialExpressionComposite_7 {class=MaterialExpressionComposite, inputs [], properties {subgraphName=TriangleUV/Radius/Sides}}
  MaterialExpressionConstant2Vector_1 {class=MaterialExpressionConstant2Vector, properties {g=1, r=0.5}}
  MaterialExpressionConstant4Vector_0 {class=MaterialExpressionConstant4Vector, inputs [], properties {constant=<rgba 1 0 0.0124 1>}}
  MaterialExpressionConstant4Vector_1 {class=MaterialExpressionConstant4Vector, inputs [], properties {constant=<rgba 1 1 1 0>}}
  MaterialExpressionConstant_10 {class=MaterialExpressionConstant, properties {r=0.25}}
  MaterialExpressionConstant_11 {class=MaterialExpressionConstant, properties {r=0.5}}
  MaterialExpressionConstant_13 {class=MaterialExpressionConstant}
  MaterialExpressionConstant_14 {class=MaterialExpressionConstant}
  MaterialExpressionConstant_23 {class=MaterialExpressionConstant, properties {r=0.25}}
  MaterialExpressionConstant_24 {class=MaterialExpressionConstant, properties {r=6}}
  MaterialExpressionConstant_25 {class=MaterialExpressionConstant, properties {r=0.5}}
  MaterialExpressionConstant_26 {class=MaterialExpressionConstant, properties {r=3}}
  MaterialExpressionConstant_27 {class=MaterialExpressionConstant, properties {r=0.25}}
  MaterialExpressionConstant_28 {class=MaterialExpressionConstant, properties {r=6}}
  MaterialExpressionConstant_29 {class=MaterialExpressionConstant, properties {r=0.5}}
  MaterialExpressionConstant_30 {class=MaterialExpressionConstant, properties {r=3}}
  MaterialExpressionConstant_8 {class=MaterialExpressionConstant, properties {r=6}}
  MaterialExpressionConstant_9 {class=MaterialExpressionConstant, properties {r=3}}
  MaterialExpressionLinearInterpolate_1 {
   class=MaterialExpressionLinearInterpolate
   inputs [
    A {from=MaterialExpressionConstant4Vector_1, output=RGBA}
    B {from=MaterialExpressionConstant4Vector_0, output=RGBA}
    Alpha {from=MaterialExpressionScalarParameter_4}
   ]
  }
  MaterialExpressionLinearInterpolate_4 {
   class=MaterialExpressionLinearInterpolate
   inputs [
    A {from=MaterialExpressionMax_5}
    B {from=MaterialExpressionMax_6}
    Alpha {from=MaterialExpressionScalarParameter_5}
   ]
  }
  MaterialExpressionMaterialFunctionCall_0 {
   class=MaterialExpressionMaterialFunctionCall
   inputs [
    UVs {from=MaterialExpressionMaterialFunctionCall_23, output=UVs}
    Radius {from=""}
    Sides {from=MaterialExpressionReroute_1, output=Sides}
    InnerAngle {from=""}
   ]
   properties {materialFunction=/FortUILibrary/MaterialFunctions/Shapes/SDFs/MF_SDF_RegularPolygon}
  }
  MaterialExpressionMaterialFunctionCall_1 {
   class=MaterialExpressionMaterialFunctionCall
   inputs [
    UVs {from=MaterialExpressionMaterialFunctionCall_24, output=UVs}
    Radius {from=MaterialExpressionReroute_1, output=Radius}
    Sides {from=MaterialExpressionReroute_2, output=Sides}
    InnerAngle {from=""}
   ]
   properties {materialFunction=&MF_SDF_RegularPolygon}
  }
  MaterialExpressionMaterialFunctionCall_10 {
   class=MaterialExpressionMaterialFunctionCall
   inputs [
    UVs {from=MaterialExpressionMaterialFunctionCall_18, output=UVs}
    Radius {from=MaterialExpressionReroute_4, output=Radius}
    Sides {from=MaterialExpressionReroute_5, output=Sides}
    InnerAngle {from=""}
   ]
   properties {materialFunction=&MF_SDF_RegularPolygon}
  }
  MaterialExpressionMaterialFunctionCall_11 {
   class=MaterialExpressionMaterialFunctionCall
   inputs [
    DistanceField {from=MaterialExpressionMax_2}
    Threshold {from=""}
    Smoothing {from=""}
   ]
   properties {materialFunction=&MF_UI_AntiAliasing}
  }
  MaterialExpressionMaterialFunctionCall_12 {
   class=MaterialExpressionMaterialFunctionCall
   inputs [
    DistanceField {from=MaterialExpressionMax_2}
    Threshold {from=""}
    Thickness {from=MaterialExpressionScalarParameter_11}
    Alignment {from=MaterialExpressionScalarParameter_9}
    Smoothing {from=""}
   ]
   properties {materialFunction=/FortUILibrary/MaterialFunctions/Shapes/Modifiers/MF_UI_ShapeOutline}
  }
  MaterialExpressionMaterialFunctionCall_14 {
   class=MaterialExpressionMaterialFunctionCall
   inputs [
    UV {from=""}
    "Arc Start Angle (0-1)" {from=""}
    "Arc End Angle (0-1)" {from=""}
    "Fill Ratio (0-1)" {from=MaterialExpressionScalarParameter_12}
    Fill Clockwise {from=""}
   ]
   properties {materialFunction=/FortUILibrary/MaterialFunctions/Basics/MF_UI_RadialArc}
  }
  MaterialExpressionMaterialFunctionCall_15 {
   class=MaterialExpressionMaterialFunctionCall
   inputs [
    UVs {from=MaterialExpressionMaterialFunctionCall_25, output=UVs}
    Sides {from=""}
   ]
   properties {materialFunction=/FortUILibrary/MaterialFunctions/Basics/MF_UI_UVRadialMirror}
  }
  MaterialExpressionMaterialFunctionCall_16 {
   class=MaterialExpressionMaterialFunctionCall
   inputs [
    UVs {from=MaterialExpressionRotator_6}
    Size {from=MaterialExpressionConstant2Vector_1}
    CornerRoundness {from=""}
    OutlineThickness {from=""}
    OutlineAlignment {from=""}
    GlowMin {from=""}
    GlowMax {from=""}
    KeepAspectRatio {from=""}
    PixelSize {from=""}
   ]
   properties {materialFunction=&MF_UI_Box}
  }
  MaterialExpressionMaterialFunctionCall_17 {
   class=MaterialExpressionMaterialFunctionCall
   inputs [
    UVs {from=MaterialExpressionReroute_0, output=UVs}
    Texture Scale {from=MaterialExpressionScalarParameter_20}
   ]
   properties {materialFunction=&ScaleUVsByCenter}
  }
  MaterialExpressionMaterialFunctionCall_18 {
   class=MaterialExpressionMaterialFunctionCall
   inputs [
    UVs {from=MaterialExpressionReroute_3, output=UV}
    Texture Scale {from=MaterialExpressionScalarParameter_20}
   ]
   properties {materialFunction=&ScaleUVsByCenter}
  }
  MaterialExpressionMaterialFunctionCall_2 {
   class=MaterialExpressionMaterialFunctionCall
   inputs [
    DistanceField {from=MaterialExpressionMax_0}
    Threshold {from=""}
    Smoothing {from=""}
   ]
   properties {materialFunction=&MF_UI_AntiAliasing}
  }
  MaterialExpressionMaterialFunctionCall_21 {
   class=MaterialExpressionMaterialFunctionCall
   inputs [
    UVs {from=MaterialExpressionReroute_0, output=UVs}
    Texture Scale {from=MaterialExpressionScalarParameter_21}
   ]
   properties {materialFunction=&ScaleUVsByCenter}
  }
  MaterialExpressionMaterialFunctionCall_22 {
   class=MaterialExpressionMaterialFunctionCall
   inputs [
    UVs {from=MaterialExpressionReroute_0, output=UV}
    Texture Scale {from=MaterialExpressionScalarParameter_21}
   ]
   properties {materialFunction=&ScaleUVsByCenter}
  }
  MaterialExpressionMaterialFunctionCall_23 {
   class=MaterialExpressionMaterialFunctionCall
   inputs [
    UVs {from=MaterialExpressionReroute_0, output=UVs}
    Texture Scale {from=MaterialExpressionScalarParameter_22}
   ]
   properties {materialFunction=&ScaleUVsByCenter}
  }
  MaterialExpressionMaterialFunctionCall_24 {
   class=MaterialExpressionMaterialFunctionCall
   inputs [
    UVs {from=MaterialExpressionReroute_0, output=UV}
    Texture Scale {from=MaterialExpressionScalarParameter_22}
   ]
   properties {materialFunction=&ScaleUVsByCenter}
  }
  MaterialExpressionMaterialFunctionCall_25 {
   class=MaterialExpressionMaterialFunctionCall
   inputs [
    UVs {from=MaterialExpressionRotator_7}
    Texture Scale {from=MaterialExpressionScalarParameter_23}
   ]
   properties {materialFunction=&ScaleUVsByCenter}
  }
  MaterialExpressionMaterialFunctionCall_3 {
   class=MaterialExpressionMaterialFunctionCall
   inputs [
    DistanceField {from=MaterialExpressionMax_0}
    Threshold {from=""}
    Thickness {from=MaterialExpressionScalarParameter_2}
    Alignment {from=MaterialExpressionScalarParameter_3}
    Smoothing {from=""}
   ]
   properties {materialFunction=&MF_UI_ShapeOutline}
  }
  MaterialExpressionMaterialFunctionCall_5 {
   class=MaterialExpressionMaterialFunctionCall
   inputs [
    UVs {from=MaterialExpressionMaterialFunctionCall_21, output=UVs}
    Radius {from=""}
    Sides {from=MaterialExpressionReroute_1, output=Sides}
    InnerAngle {from=""}
   ]
   properties {materialFunction=&MF_SDF_RegularPolygon}
  }
  MaterialExpressionMaterialFunctionCall_6 {
   class=MaterialExpressionMaterialFunctionCall
   inputs [
    UVs {from=MaterialExpressionMaterialFunctionCall_22, output=UVs}
    Radius {from=MaterialExpressionReroute_1, output=Radius}
    Sides {from=MaterialExpressionReroute_2, output=Sides}
    InnerAngle {from=""}
   ]
   properties {materialFunction=&MF_SDF_RegularPolygon}
  }
  MaterialExpressionMaterialFunctionCall_7 {
   class=MaterialExpressionMaterialFunctionCall
   inputs [
    DistanceField {from=MaterialExpressionMax_1}
    Threshold {from=""}
    Smoothing {from=""}
   ]
   properties {materialFunction=&MF_UI_AntiAliasing}
  }
  MaterialExpressionMaterialFunctionCall_8 {
   class=MaterialExpressionMaterialFunctionCall
   inputs [
    DistanceField {from=MaterialExpressionMax_1}
    Threshold {from=""}
    Thickness {from=MaterialExpressionScalarParameter_7}
    Alignment {from=MaterialExpressionScalarParameter_6}
    Smoothing {from=""}
   ]
   properties {materialFunction=&MF_UI_ShapeOutline}
  }
  MaterialExpressionMaterialFunctionCall_9 {
   class=MaterialExpressionMaterialFunctionCall
   inputs [
    UVs {from=MaterialExpressionMaterialFunctionCall_17, output=UVs}
    Radius {from=""}
    Sides {from=MaterialExpressionReroute_1, output=Sides}
    InnerAngle {from=""}
   ]
   properties {materialFunction=&MF_SDF_RegularPolygon}
  }
  MaterialExpressionMax_0 {
   class=MaterialExpressionMax
   inputs [
    A {from=MaterialExpressionMaterialFunctionCall_1, output=SDF}
    B {from=MaterialExpressionMaterialFunctionCall_0, output=SDF}
   ]
  }
  MaterialExpressionMax_1 {
   class=MaterialExpressionMax
   inputs [
    A {from=MaterialExpressionMaterialFunctionCall_6, output=SDF}
    B {from=MaterialExpressionMaterialFunctionCall_5, output=SDF}
   ]
  }
  MaterialExpressionMax_2 {
   class=MaterialExpressionMax
   inputs [
    A {from=MaterialExpressionMaterialFunctionCall_10, output=SDF}
    B {from=MaterialExpressionMaterialFunctionCall_9, output=SDF}
   ]
  }
  MaterialExpressionMax_5 {
   class=MaterialExpressionMax
   inputs [
    A {from=MaterialExpressionNamedRerouteUsage_10}
    B {from=MaterialExpressionNamedRerouteUsage_11}
   ]
  }
  MaterialExpressionMax_6 {
   class=MaterialExpressionMax
   inputs [
    A {from=MaterialExpressionNamedRerouteUsage_12}
    B {from=MaterialExpressionMultiply_4}
   ]
  }
  MaterialExpressionMultiply_12 {
   class=MaterialExpressionMultiply
   inputs [
    A {from=MaterialExpressionStaticSwitch_1}
    B {from=MaterialExpressionScalarParameter_14}
   ]
  }
  MaterialExpressionMultiply_2 {
   class=MaterialExpressionMultiply
   inputs [
    A {from=MaterialExpressionComponentMask_5}
    B {from=MaterialExpressionOneMinus_0}
   ]
  }
  MaterialExpressionMultiply_3 {
   class=MaterialExpressionMultiply
   inputs [
    A {from=MaterialExpressionLinearInterpolate_4}
    B {from=MaterialExpressionScalarParameter_10}
   ]
  }
  MaterialExpressionMultiply_4 {
   class=MaterialExpressionMultiply
   inputs [
    A {from=MaterialExpressionNamedRerouteUsage_13}
    B {from=MaterialExpressionMax_5}
   ]
  }
  MaterialExpressionMultiply_7 {
   class=MaterialExpressionMultiply
   inputs [
    A {from=MaterialExpressionStaticSwitch_3}
    B {from=MaterialExpressionOneMinus_1}
   ]
  }
  MaterialExpressionMultiply_8 {
   class=MaterialExpressionMultiply
   inputs [
    A {from=MaterialExpressionMultiply_7}
    B {from=MaterialExpressionLinearInterpolate_1}
   ]
  }
  MaterialExpressionNamedRerouteDeclaration_0 {
   class=MaterialExpressionNamedRerouteDeclaration
   inputs [None {from=MaterialExpressionMultiply_12}]
   properties {name=BaseCrosshair, nodeColor=<rgba 0.8047 1 0 1>}
  }
  MaterialExpressionNamedRerouteDeclaration_1 {
   class=MaterialExpressionNamedRerouteDeclaration
   inputs [None {from=MaterialExpressionMultiply_8}]
   properties {name=AccentCrosshair, nodeColor=<rgba 0.875 0 1 1>}
  }
  MaterialExpressionNamedRerouteDeclaration_2 {
   class=MaterialExpressionNamedRerouteDeclaration
   inputs [None {from=MaterialExpressionMultiply_2}]
   properties {name=TimerCrosshair, nodeColor=<rgba 0 1 0.5547 1>}
  }
  MaterialExpressionNamedRerouteDeclaration_3 {
   class=MaterialExpressionNamedRerouteDeclaration
   inputs [None {from=MaterialExpressionMaterialFunctionCall_16, output=Fill}]
   properties {name=AccentMask, nodeColor=<rgba 1 0.2344 0 1>}
  }
  MaterialExpressionNamedRerouteUsage_10 {class=MaterialExpressionNamedRerouteUsage}
  MaterialExpressionNamedRerouteUsage_11 {class=MaterialExpressionNamedRerouteUsage}
  MaterialExpressionNamedRerouteUsage_12 {class=MaterialExpressionNamedRerouteUsage}
  MaterialExpressionNamedRerouteUsage_13 {class=MaterialExpressionNamedRerouteUsage}
  MaterialExpressionNamedRerouteUsage_2 {class=MaterialExpressionNamedRerouteUsage}
  MaterialExpressionOneMinus_0 {class=MaterialExpressionOneMinus, inputs [None {from=MaterialExpressionCeil_0}]}
  MaterialExpressionOneMinus_1 {class=MaterialExpressionOneMinus, inputs [None {from=MaterialExpressionNamedRerouteUsage_2}]}
  MaterialExpressionPinBase_0 {class=MaterialExpressionPinBase, inputs [], properties {}}
  MaterialExpressionPinBase_1 {
   class=MaterialExpressionPinBase
   inputs [
    UVs {from=MaterialExpressionRotator_4, output=""}
    Sides {from=MaterialExpressionConstant_8, output=""}
   ]
   properties {reroutePins [
     UVs {expression=&self:MaterialGraph_1.MaterialGraphNode_Composite_0.HexUV/Sides.MaterialExpressionReroute_0}
     Sides {expression=&self:MaterialGraph_1.MaterialGraphNode_Composite_0.HexUV/Sides.MaterialExpressionReroute_1}
    ]}
  }
  MaterialExpressionPinBase_10 {class=MaterialExpressionPinBase, inputs [], properties {}}
  MaterialExpressionPinBase_11 {
   class=MaterialExpressionPinBase
   inputs [
    UV {from=MaterialExpressionRotator_11, output=""}
    Radius {from=MaterialExpressionAdd_12, output=""}
    Sides {from=MaterialExpressionConstant_26, output=""}
   ]
   properties {reroutePins [
     UV {expression=&self:MaterialGraph_1.MaterialGraphNode_Composite_3.TriangleUV/Radius/Sides.MaterialExpressionReroute_0}
     Radius {expression=&self:MaterialGraph_1.MaterialGraphNode_Composite_3.TriangleUV/Radius/Sides.MaterialExpressionReroute_1}
     Sides {expression=&self:MaterialGraph_1.MaterialGraphNode_Composite_3.TriangleUV/Radius/Sides.MaterialExpressionReroute_2}
    ]}
  }
  MaterialExpressionPinBase_12 {class=MaterialExpressionPinBase, inputs [], properties {}}
  MaterialExpressionPinBase_13 {
   class=MaterialExpressionPinBase
   inputs [
    UVs {from=MaterialExpressionRotator_12, output=""}
    Sides {from=MaterialExpressionConstant_28, output=""}
   ]
   properties {reroutePins [
     UVs {expression=&self:MaterialGraph_1.MaterialGraphNode_Composite_4.HexUV/Sides.MaterialExpressionReroute_0}
     Sides {expression=&self:MaterialGraph_1.MaterialGraphNode_Composite_4.HexUV/Sides.MaterialExpressionReroute_1}
    ]}
  }
  MaterialExpressionPinBase_14 {class=MaterialExpressionPinBase, inputs [], properties {}}
  MaterialExpressionPinBase_15 {
   class=MaterialExpressionPinBase
   inputs [
    UV {from=MaterialExpressionRotator_13, output=""}
    Radius {from=MaterialExpressionAdd_13, output=""}
    Sides {from=MaterialExpressionConstant_30, output=""}
   ]
   properties {reroutePins [
     UV {expression=&self:MaterialGraph_1.MaterialGraphNode_Composite_5.TriangleUV/Radius/Sides.MaterialExpressionReroute_0}
     Radius {expression=&self:MaterialGraph_1.MaterialGraphNode_Composite_5.TriangleUV/Radius/Sides.MaterialExpressionReroute_1}
     Sides {expression=&self:MaterialGraph_1.MaterialGraphNode_Composite_5.TriangleUV/Radius/Sides.MaterialExpressionReroute_2}
    ]}
  }
  MaterialExpressionPinBase_2 {class=MaterialExpressionPinBase, inputs [], properties {}}
  MaterialExpressionPinBase_3 {
   class=MaterialExpressionPinBase
   inputs [
    UV {from=MaterialExpressionRotator_5, output=""}
    Radius {from=MaterialExpressionAdd_2, output=""}
    Sides {from=MaterialExpressionConstant_9, output=""}
   ]
   properties {reroutePins [
     UV {expression=&self:MaterialGraph_1.MaterialGraphNode_Composite_1.TriangleUV/Radius/Sides.MaterialExpressionReroute_3}
     Radius {expression=&self:MaterialGraph_1.MaterialGraphNode_Composite_1.TriangleUV/Radius/Sides.MaterialExpressionReroute_4}
     Sides {expression=&self:MaterialGraph_1.MaterialGraphNode_Composite_1.TriangleUV/Radius/Sides.MaterialExpressionReroute_5}
    ]}
  }
  MaterialExpressionPinBase_8 {class=MaterialExpressionPinBase, inputs [], properties {}}
  MaterialExpressionPinBase_9 {
   class=MaterialExpressionPinBase
   inputs [
    UVs {from=MaterialExpressionRotator_10, output=""}
    Sides {from=MaterialExpressionConstant_24, output=""}
   ]
   properties {reroutePins [
     UVs {expression=&self:MaterialGraph_1.MaterialGraphNode_Composite_2.HexUV/Sides.MaterialExpressionReroute_0}
     Sides {expression=&self:MaterialGraph_1.MaterialGraphNode_Composite_2.HexUV/Sides.MaterialExpressionReroute_1}
    ]}
  }
  MaterialExpressionRotator_10 {
   class=MaterialExpressionRotator
   inputs [
    Coordinate {from="", output=""}
    Time {from=MaterialExpressionConstant_23, output=""}
   ]
   properties {speed=6.2832}
  }
  MaterialExpressionRotator_11 {
   class=MaterialExpressionRotator
   inputs [
    Coordinate {from="", output=""}
    Time {from=MaterialExpressionConstant_25, output=""}
   ]
   properties {speed=6.2832}
  }
  MaterialExpressionRotator_12 {
   class=MaterialExpressionRotator
   inputs [
    Coordinate {from="", output=""}
    Time {from=MaterialExpressionConstant_27, output=""}
   ]
   properties {speed=6.2832}
  }
  MaterialExpressionRotator_13 {
   class=MaterialExpressionRotator
   inputs [
    Coordinate {from="", output=""}
    Time {from=MaterialExpressionConstant_29, output=""}
   ]
   properties {speed=6.2832}
  }
  MaterialExpressionRotator_4 {
   class=MaterialExpressionRotator
   inputs [
    Coordinate {from="", output=""}
    Time {from=MaterialExpressionConstant_10, output=""}
   ]
   properties {speed=6.2832}
  }
  MaterialExpressionRotator_5 {
   class=MaterialExpressionRotator
   inputs [
    Coordinate {from="", output=""}
    Time {from=MaterialExpressionConstant_11, output=""}
   ]
   properties {speed=6.2832}
  }
  MaterialExpressionRotator_6 {
   class=MaterialExpressionRotator
   inputs [
    Coordinate {from=MaterialExpressionComponentMask_7, output=""}
    Time {from=MaterialExpressionConstant_13, output=""}
   ]
   properties {speed=6.2832}
  }
  MaterialExpressionRotator_7 {
   class=MaterialExpressionRotator
   inputs [
    Coordinate {from="", output=""}
    Time {from=MaterialExpressionConstant_14, output=""}
   ]
   properties {speed=6.2832}
  }
  MaterialExpressionSaturate_0 {class=MaterialExpressionSaturate, inputs [None {from=MaterialExpressionMaterialFunctionCall_14, output=Centered Fill}]}
  MaterialExpressionScalarParameter_10 {class=MaterialExpressionScalarParameter, properties {defaultValue=0.8, parameterName=ReticleOpacity}}
  MaterialExpressionScalarParameter_11 {class=MaterialExpressionScalarParameter, properties {defaultValue=0.025, parameterName=OutlineThicknessTimer}}
  MaterialExpressionScalarParameter_12 {class=MaterialExpressionScalarParameter, properties {parameterName=CooldownProgress}}
  MaterialExpressionScalarParameter_14 {class=MaterialExpressionScalarParameter, properties {defaultValue=0.8, parameterName=ReticleOpacity}}
  MaterialExpressionScalarParameter_18 {class=MaterialExpressionScalarParameter, properties {defaultValue=0.2, parameterName=ChamferAmt}}
  MaterialExpressionScalarParameter_19 {class=MaterialExpressionScalarParameter, properties {defaultValue=0.2, parameterName=ChamferAmt}}
  MaterialExpressionScalarParameter_2 {class=MaterialExpressionScalarParameter, properties {defaultValue=0.0075, parameterName=OutlineThicknessBase}}
  MaterialExpressionScalarParameter_20 {class=MaterialExpressionScalarParameter, properties {defaultValue=0.55, parameterName=ReticleScale}}
  MaterialExpressionScalarParameter_21 {class=MaterialExpressionScalarParameter, properties {defaultValue=0.55, parameterName=ReticleScale}}
  MaterialExpressionScalarParameter_22 {class=MaterialExpressionScalarParameter, properties {defaultValue=0.55, parameterName=ReticleScale}}
  MaterialExpressionScalarParameter_23 {class=MaterialExpressionScalarParameter, properties {defaultValue=0.55, parameterName=ReticleScale}}
  MaterialExpressionScalarParameter_3 {class=MaterialExpressionScalarParameter, properties {parameterName=OutlineAlignment}}
  MaterialExpressionScalarParameter_4 {class=MaterialExpressionScalarParameter, properties {defaultValue=1, parameterName=ColorFlash}}
  MaterialExpressionScalarParameter_5 {class=MaterialExpressionScalarParameter, properties {defaultValue=1, parameterName=Ready?, sliderMax=1}}
  MaterialExpressionScalarParameter_6 {class=MaterialExpressionScalarParameter, properties {parameterName=OutlineAlignment}}
  MaterialExpressionScalarParameter_7 {class=MaterialExpressionScalarParameter, properties {defaultValue=0.025, parameterName=OutlineThicknessAccent}}
  MaterialExpressionScalarParameter_8 {class=MaterialExpressionScalarParameter, properties {defaultValue=0.2, parameterName=ChamferAmt}}
  MaterialExpressionScalarParameter_9 {class=MaterialExpressionScalarParameter, properties {parameterName=OutlineAlignment}}
  MaterialExpressionStaticBool_0 {class=MaterialExpressionStaticBool}
  MaterialExpressionStaticBool_1 {class=MaterialExpressionStaticBool}
  MaterialExpressionStaticBool_2 {class=MaterialExpressionStaticBool}
  MaterialExpressionStaticBool_3 {class=MaterialExpressionStaticBool}
  MaterialExpressionStaticSwitch_0 {
   class=MaterialExpressionStaticSwitch
   inputs [
    True {from=MaterialExpressionAppendVector_1, output=""}
    False {from=MaterialExpressionAppendVector_0, output=""}
    Value {from=MaterialExpressionStaticBool_0, output=""}
   ]
   properties {}
  }
  MaterialExpressionStaticSwitch_1 {
   class=MaterialExpressionStaticSwitch
   inputs [
    True {from=MaterialExpressionComponentMask_0, output=""}
    False {from=MaterialExpressionComponentMask_1, output=""}
    Value {from=MaterialExpressionStaticBool_1, output=""}
   ]
   properties {}
  }
  MaterialExpressionStaticSwitch_2 {
   class=MaterialExpressionStaticSwitch
   inputs [
    True {from=MaterialExpressionAppendVector_3, output=""}
    False {from=MaterialExpressionAppendVector_2, output=""}
    Value {from=MaterialExpressionStaticBool_2, output=""}
   ]
   properties {}
  }
  MaterialExpressionStaticSwitch_3 {
   class=MaterialExpressionStaticSwitch
   inputs [
    True {from=MaterialExpressionComponentMask_2, output=""}
    False {from=MaterialExpressionComponentMask_3, output=""}
    Value {from=MaterialExpressionStaticBool_3, output=""}
   ]
   properties {}
  }
 }
 properties {
  bAutomaticallySetUsageInEditor=false
  blendMode=BLEND_Additive
  materialDomain=MD_UI
  pixelDepthOffsetMode=PDOM_Legacy
  usedShadingModels=Default Lit
 }
}
