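# Additive UI material for the BarnTowel cooldown indicator, showing a Progress-driven fill with a hazy distortion.
# A reflected gradient thresholded against inverted, scaled Progress makes a hard mask that brightens the cooldown texture's green channel.
# This combines with the red channel, less a fraction of green, and is tinted by Color.
# Two opposing panned noise textures, scaled by Haze, form the GlitchedUVs reroute that warps texture and gradient lookups.
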Material /BarnTowelGameplay/UI/V2/M_UI_Barntowel_Cooldown
graph {
 properties {bAutomaticallySetUsageInEditor=false, blendMode=BLEND_Additive, materialDomain=MD_UI, usedShadingModels=Default Lit}
 nodes [
  n1 = TextureSample(UVs=NamedRerouteUsage()) {samplerType=SAMPLERTYPE_LinearColor, texture=&/T_UI_BarnTowel_CoolDown}
  n2 = Multiply(A=@n1.G) {constB=0.25}
  n3 = NamedRerouteUsage()
  n4 = ScalarParameter() {parameterName=Progress}
  n5 = Constant()
  n6 = If(
   A=OneMinus(None=MaterialFunctionCall!(UVs=@n3, Offset, RotationCenter, RotationAngle) {materialFunction=/FortUILibrary/MaterialFunctions/Gradients/MF_UI_ReflectedGradient}.Result)
   B=Multiply(A=OneMinus(None=@n4)) {constB=0.64}
   "A > B"=@n5
   "A == B"=@n5
   "A < B"=Constant() {r=1}
  )
  Multiply(
   A=Add(A=Subtract(A=@n1.R, B=@n2), B=Multiply(A=@n1.G, B=@n6))
   B=VectorParameter() {parameterName=Color, defaultValue=<rgba 0.4072 0.4852 1 1>}.RGB
  )
  n7 = TextureCoordinate()
  n8 = TextureSample(UVs=Panner(Coordinate=@n7) {speedX=0.1, speedY=0.02}) {texture=/Game/Packages/Fortress_Transylvania/Haunted_Forest/Swamps/Textures/2ndInfectedMapLinear2}
  n9 = TextureSample(UVs=Panner(Coordinate=Multiply(A=@n7) {constB=0.76}) {speedX=-0.05, speedY=-0.01}) {texture=&2ndInfectedMapLinear2}
  n10 = ConstantBiasScale(None=ComponentMask(None=Multiply(A=@n8.RGB, B=@n9.RGB)) {g=true, r=true}) {bias=-0.125, scale=2}
  n11 = ScalarParameter() {parameterName=Haze, sliderMax=0.2}
  NamedRerouteDeclaration(None=Add(A=TextureCoordinate(), B=Multiply(A=@n10, B=@n11) {constB=0.2})) {name=GlitchedUVs, nodeColor=<rgba 0 0 0 1>}
 ]
}
