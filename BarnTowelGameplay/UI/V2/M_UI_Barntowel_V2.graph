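# Additive UI material for the BarnTowel V2 HUD, drawing a tinted base graphic, an attitude indicator and a radial speed gauge.
# Pitch rotates the attitude texture, gated by ShowFrame and a radial mask, while Speed fills the gauge, which pulses red to orange in Overdrive.
# Named reroutes define UV layers, including Haze-scaled distortion from two panning noise samples and a clamped drift offset from CameraVelocity.
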Material /BarnTowelGameplay/UI/V2/M_UI_Barntowel_V2
graph {
 properties {bAutomaticallySetUsageInEditor=false, blendMode=BLEND_Additive, materialDomain=MD_UI, usedShadingModels=Default Lit}
 nodes [
  n1 = NamedRerouteUsage()
  n2 = NamedRerouteUsage()
  n3 = Constant() {r=0.55}
  n4 = Constant() {r=0.2}
  n5 = ScalarParameter() {parameterName=ShowFrame?}
  n6 = OneMinus(None=Multiply(
    A=MaterialFunctionCall(UVs=@n2, Size=@n3, OutlineThickness=@n4) {materialFunction=&MF_UI_Circle}.Outline
    B=OneMinus(None=@n5)
   ))
  n7 = TextureSample(UVs=NamedRerouteUsage()) {samplerType=SAMPLERTYPE_LinearColor, texture=&/T_UI_BarnTowel_Base_V2}
  n8 = Saturate(None=Subtract(A=@n7.R, B=@n7.R))
  n9 = Constant() {r=0.665}
  n10 = MaterialFunctionCall(UVs=NamedRerouteUsage(), Texture Scale=AppendVector(A=@n9, B=Multiply(A=@n9) {constB=1.7})) {materialFunction=&ScaleUVsByCenter}
  n11 = Multiply(A=ScalarParameter() {parameterName=Pitch}) {constB=0.0005}
  n12 = Multiply(
   A=TextureSample(UVs=MaterialFunctionCall(UVs=@n10.UVs, "Rotation Angle (0-1)"=@n11) {materialFunction=&CustomRotator}.Rotated Values) {samplerType=SAMPLERTYPE_LinearGrayscale, texture=&/T_UI_BarnTowel_Attitude_V2}.R
   B=ScalarParameter() {parameterName=ShowFrame?}
  )
  n13 = MaterialFunctionCall!(UVs=NamedRerouteUsage(), Progress=Constant() {r=1}, StartValue=Constant() {r=102}, Range=Constant() {r=-24}) {materialFunction=/FortUILibrary/MaterialFunctions/Basics/MF_UI_RadialProgress}
  n14 = Multiply(
   A=Add(A=Multiply(A=@n6, B=@n8), B=Multiply(A=@n12, B=@n13.Result))
   B=VectorParameter() {parameterName=Color, defaultValue=<rgba 0.4072 0.4852 1 1>}.RGB
  )
  n15 = OneMinus(None=MaterialFunctionCall(Radius=Constant() {r=1}, Density=Constant() {r=1.5}) {materialFunction=&RadialGradientExponential}.RadialGradientExponential)
  n16 = Multiply(
   A=@n7.B
   B=MaterialFunctionCall(
    UVs=NamedRerouteUsage()
    Progress=ScalarParameter() {defaultValue=1, parameterName=Speed}
    StartValue=Constant() {r=258}
    Range=Constant() {r=24}
   ) {materialFunction=&MF_UI_RadialProgress}.Result
  )
  n17 = Multiply(A=Time()) {constB=4}
  n18 = LinearInterpolate(
   A=VectorParameter() {parameterName=Color, defaultValue=<rgba 0.4072 0.4852 1 1>}.RGB
   B=LinearInterpolate(
    A=Constant3Vector() {constant=<rgba 1 0 0 0>}
    B=Constant3Vector() {constant=<rgba 1 0.5615 0 1>}
    Alpha=MaterialFunctionCall!(Sine Phase=@n17, Value 1, Value 2) {materialFunction=/Engine/Functions/Engine_MaterialFunctions02/Utility/Sine_Remapped}.Remapped Sine
   )
   Alpha=ScalarParameter() {parameterName=Overdrive?}
  )
  Add(
   A=Multiply(
    A=TextureSample(UVs=@n1) {samplerType=SAMPLERTYPE_LinearColor, texture=&T_UI_BarnTowel_Base_V2}.G
    B=VectorParameter() {parameterName=Color, defaultValue=<rgba 0.4072 0.4852 1 1>}.RGB
   )
   B=Add(A=Multiply(A=@n14, B=@n15), B=Multiply(A=@n16, B=@n18))
  )
  n19 = TextureCoordinate()
  n20 = TextureSample(UVs=Panner(Coordinate=@n19) {speedX=0.1, speedY=0.02}) {texture=/Game/Packages/Fortress_Transylvania/Haunted_Forest/Swamps/Textures/2ndInfectedMapLinear2}
  n21 = TextureSample(UVs=Panner(Coordinate=Multiply(A=@n19) {constB=0.76}) {speedX=-0.05, speedY=-0.01}) {texture=&2ndInfectedMapLinear2}
  n22 = ConstantBiasScale(None=ComponentMask(None=Multiply(A=@n20.RGB, B=@n21.RGB)) {g=true, r=true}) {bias=-0.125, scale=2}
  n23 = ScalarParameter() {parameterName=Haze, sliderMax=0.2}
  NamedRerouteDeclaration(None=Add(A=TextureCoordinate(), B=Multiply(A=@n22, B=@n23) {constB=0.2})) {name=GlitchedUVs, nodeColor=<rgba 0 0 0 1>}
  NamedRerouteDeclaration(None=Add(A=NamedRerouteUsage(), B=NamedRerouteUsage())) {name=FlavorUVs, nodeColor=<rgba 0 0 0 1>}
  n24 = NamedRerouteUsage()
  NamedRerouteDeclaration(None=Add(A=NamedRerouteUsage(), B=Divide(A=@n24) {constB=1.2})) {name=DotsUV, nodeColor=<rgba 0 0 0 1>}
  n25 = ComponentMask(None=VectorParameter() {parameterName=CameraVelocity}.RGB) {g=true, r=true}
  n26 = Constant2Vector() {g=0.84, r=9.6}
  n27 = Constant2Vector() {desc=Max Drift, g=0.35, r=0.09}
  NamedRerouteDeclaration(None=Clamp(None=Divide(A=@n25, B=@n26), Min=Multiply(A=@n27) {constB=-1}, Max=@n27)) {name=CameraOffset, nodeColor=<rgba 0 0 0 1>}
 ]
}
