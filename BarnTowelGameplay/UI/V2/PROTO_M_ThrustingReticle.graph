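# Prototype additive UI material that draws a thrusting reticle of triangle markers around a segmented ring.
# Two mirrored pairs of SDF triangle outlines rotate in opposite directions through RotationSwitch, with TriangleSpread offsetting them radially.
# A four-segment circle outline with size and opacity parameters is max-blended with the triangles, then scaled by MasterOpacity.
# Older circle, triangle and texture-based experiments remain alongside an "Old Reticle Work Archive" composite.
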Material /BarnTowelGameplay/UI/V2/PROTO_M_ThrustingReticle
graph {
 properties {blendMode=BLEND_Additive, materialDomain=MD_UI, pixelDepthOffsetMode=PDOM_Legacy, usedShadingModels=Default Lit}
 nodes [
  Composite() {subgraphName=Old Reticle Work Archive}
  Constant() {r=0.125}
  n1 = ScalarParameter() {defaultValue=1, parameterName=Circle Start}
  n2 = MaterialFunctionCall(
   Size=ScalarParameter() {defaultValue=0.9, group=1 - Circle, parameterName=Circle Size, sliderMax=1}
   OutlineThickness=ScalarParameter() {defaultValue=0.01, group=1 - Circle, parameterName=Circle Thickness, sliderMax=1}
  ) {materialFunction=&MF_UI_Circle}
  n3 = SmoothStep(
   Min=Subtract(A=@n1, B=ScalarParameter() {defaultValue=0.07, parameterName=Circle Feather}) {constB=0.05}
   Max=@n1
   Value=Saturate(None=OneMinus(None=@n2.SDFOutline))
  )
  n4 = Constant() {r=4}
  n5 = ScalarParameter() {defaultValue=0.4, parameterName=Segment Spacing}
  n6 = Saturate(None=Multiply(A=MaterialFunctionCall!(UV, Segment Count=@n4, Segment Spacing=@n5, Fixed Spacing) {materialFunction=/FortUILibrary/MaterialFunctions/Basics/MF_UI_RadialSegments}.Segmented Triangles Gradient))
  n7 = ScalarParameter() {defaultValue=0.98, parameterName=Triangle Start}
  n8 = ScalarParameter() {parameterName=Triangle Feather}
  n9 = MaterialFunctionCall(
   UVs=Add(
    A=MaterialFunctionCall!(UV, Swizzle UV) {materialFunction=/FortUILibrary/MaterialFunctions/Basics/MF_UI_Symmetry_2x2_New}.Symmetry
    B=ScalarParameter() {defaultValue=0.125, parameterName=Triangle Distance}
   )
   "Rotation Angle (0-1)"=Constant() {r=0.125}
  ) {materialFunction=&CustomRotator}
  n10 = Constant() {r=0.4}
  n11 = OneMinus(None=MaterialFunctionCall!(
    UVs=MaterialFunctionCall(UVs=@n9.Rotated Values, Texture Scale=@n10) {materialFunction=&ScaleUVsByCenter}.UVs
    Corner 01
    Corner 02
    Corner 03
    Glow Min
    Glow Max
    Outline Thickness
   ) {materialFunction=/FortUILibrary/MaterialFunctions/Shapes/MF_UI_WIP_Triangle}.SDF)
  n12 = Add(A=Saturate(None=Subtract(A=@n3, B=@n6)), B=SmoothStep(Min=Subtract(A=@n7, B=@n8) {constB=0.05}, Max=@n7, Value=Saturate(None=@n11)))
  n13 = ScalarParameter() {defaultValue=1, parameterName=Circle Start}
  n14 = Subtract(A=@n13, B=ScalarParameter() {defaultValue=0.07, parameterName=Circle Feather}) {constB=0.05}
  n15 = Constant() {r=0.07}
  n16 = Saturate(None=OneMinus(None=MaterialFunctionCall(Size=@n15) {materialFunction=&MF_UI_Circle}.SDF))
  LinearInterpolate(
   A=VectorParameter() {parameterName=Outer Color, defaultValue=<rgba 0 0 0 1>}.RGB
   B=VectorParameter() {parameterName=Color, defaultValue=<rgba 1 1 1 1>}.RGB
   Alpha=Add(A=SmoothStep(Min=@n14, Max=@n13, Value=@n16), B=Saturate(None=@n12))
  )
  n17 = TextureSample() {texture=&/T_UI_Thrusting}
  Multiply(A=SmoothStep(Min=Constant() {r=0.05}, Max=Constant() {r=0.4}, Value=Add(A=@n17.R, B=@n17.R))) {constB=0.06}
  n18 = MaterialFunctionCall() {materialFunction=&GetUserInterfaceUV}
  n19 = LinearInterpolate(Alpha=ScalarParameter() {defaultValue=1, parameterName=RotationSwitch}) {constB=0.125}
  n20 = Constant() {r=2}
  n21 = Constant() {r=0.75}
  n22 = MaterialFunctionCall(
   UVs=MaterialFunctionCall!(
    UVs=MaterialFunctionCall(UVs=@n18.Normalized UV, "Rotation Angle (0-1)"=@n19) {materialFunction=&CustomRotator}.Rotated Values
    Sides=@n20
   ) {materialFunction=/FortUILibrary/MaterialFunctions/Basics/MF_UI_UVRadialMirror}.UVs
   Texture Scale=@n21
  ) {materialFunction=&ScaleUVsByCenter}
  n23 = ComponentMask(None=@n22.UVs) {g=true}
  n24 = LinearInterpolate(Alpha=ScalarParameter() {parameterName=TriangleSpread}) {constA=0.1, constB=0}
  n25 = Divide() {constB=6}
  n26 = MaterialFunctionCall(UVs=AppendVector(A=ComponentMask(None=@n22.UVs) {r=true}, B=Add(A=@n23, B=@n24) {constB=-0.4}), "Rotation Angle (0-1)"=@n25) {materialFunction=&CustomRotator}
  n27 = Constant() {r=3}
  n28 = Constant() {r=0.05}
  n29 = Constant() {r=-1}
  n30 = MaterialFunctionCall(DistanceField=MaterialFunctionCall!(
    DistanceField=MaterialFunctionCall!(UVs=@n26.Rotated Values, Radius, Sides=@n27, InnerAngle) {materialFunction=/FortUILibrary/MaterialFunctions/Shapes/SDFs/MF_SDF_RegularPolygon}.SDF
    Threshold
    Thickness=@n28
    Alignment=@n29
    Smoothing
   ) {materialFunction=/FortUILibrary/MaterialFunctions/Shapes/Modifiers/MF_UI_ShapeOutline}.SDFOutline) {materialFunction=&MF_UI_AntiAliasing}
  n31 = Multiply(A=@n19) {constB=-1}
  n32 = MaterialFunctionCall(
   UVs=MaterialFunctionCall(
    UVs=MaterialFunctionCall(UVs=@n18.Normalized UV, "Rotation Angle (0-1)"=@n31) {materialFunction=&CustomRotator}.Rotated Values
    Sides=@n20
   ) {materialFunction=&MF_UI_UVRadialMirror}.UVs
   Texture Scale=@n21
  ) {materialFunction=&ScaleUVsByCenter}
  n33 = ComponentMask(None=@n32.UVs) {g=true}
  n34 = MaterialFunctionCall(UVs=AppendVector(A=ComponentMask(None=@n32.UVs) {r=true}, B=Add(A=@n33, B=@n24) {constB=-0.4}), "Rotation Angle (0-1)"=@n25) {materialFunction=&CustomRotator}
  n35 = MaterialFunctionCall(DistanceField=MaterialFunctionCall(
    DistanceField=MaterialFunctionCall(UVs=@n34.Rotated Values, Sides=@n27) {materialFunction=&MF_SDF_RegularPolygon}.SDF
    Thickness=@n28
    Alignment=@n29
   ) {materialFunction=&MF_UI_ShapeOutline}.SDFOutline) {materialFunction=&MF_UI_AntiAliasing}
  Multiply(A=Max(A=Add(A=@n30.Inner, B=@n35.Inner), B=NamedRerouteUsage()), B=ScalarParameter() {defaultValue=0.65, parameterName=MasterOpacity})
  n36 = MaterialFunctionCall(
   UV=MaterialFunctionCall(UVs=@n18.Normalized UV, "Rotation Angle (0-1)"=Constant() {r=0.25}) {materialFunction=&CustomRotator}.Rotated Values
   Segment Count=Constant() {r=4}
   Segment Spacing=Constant() {r=0.55}
  ) {materialFunction=&MF_UI_RadialSegments}
  n37 = MaterialFunctionCall(
   UVs=MaterialFunctionCall(UVs=@n18.Normalized UV) {materialFunction=&ScaleUVsByCenter}.UVs
   Size=LinearInterpolate(Alpha=ScalarParameter() {defaultValue=1, parameterName=CircleSize}) {constA=-0.1, constB=0.475}
   OutlineThickness=Constant() {r=0.025}
  ) {materialFunction=&MF_UI_Circle}
  NamedRerouteDeclaration(None=Multiply(A=Multiply(A=@n36.Segmented Triangles Fill, B=@n37.Outline), B=ScalarParameter() {defaultValue=1, parameterName=CircleOpacity})) {name=Circle, nodeColor=<rgba 1 0.7969 0 1>}
  PinBase()
  PinBase()
  ScalarParameter() {group=1 - Circle, parameterName=Spacing Falloff, sliderMax=1}
  SmoothStep(Min=Constant() {r=0.32}, Max=Constant() {r=0.5}, Value=Add(A=@n17.R, B=@n17.R))
 ]
}
