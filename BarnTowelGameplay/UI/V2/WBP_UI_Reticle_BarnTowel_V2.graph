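# Weapon reticle HUD for the BarnTowel flight item, showing flight frame, fuel bar, thrust wind-up, unibeam triangle and cooldowns.
# Async gameplay tag, event and LocalFuel listeners drive animations and material parameters, including critical-fuel warning audio.
# A trailing fuel value catches up at constant speed, slowed for a window after thrust and delay events.
# Tick updates camera-rotation and pawn-velocity effects, while a looping timer polls repulsor cooldown from active gameplay effects.
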WidgetBlueprint /BarnTowelGameplay/UI/V2/WBP_UI_Reticle_BarnTowel_V2
animations [
 Initialize_Frame {
  binding_count=1
  length_seconds=0.45
  object=&self:Initialize_Frame
  sequence {bindings [Frame {
     id=#1
     possessed_class=Image
     tracks [{
       type=MovieSceneWidgetMaterialTrack
       sections [{type=MovieSceneParameterSection, end_frame=9, start_frame=0, channels [Haze {type=MovieSceneScriptingFloatChannel, keys [[0,0.15],[6,0]]}]}]
      }]
    }]}
 }
 Initialize_Cannon {
  binding_count=1
  length_seconds=1
  object=&self:Initialize_Cannon
  sequence {bindings [triangle {
     id=#2
     possessed_class=Image
     tracks [
      {
       type=MovieSceneFloatTrack
       display_name=Render Opacity
       sections [{type=MovieSceneFloatSection, end_frame=20, start_frame=0, channels [None {type=MovieSceneScriptingFloatChannel, keys [[0,0],[10,0],[14,1]]}]}]
      }
      {
       type=MovieScene2DTransformTrack
       sections [{
         type=MovieScene2DTransformSection
         end_frame=20
         start_frame=8
         channels [
          Scale.X {type=MovieSceneScriptingFloatChannel, keys [[8,1],[10,1.25],[14,1]]}
          Scale.Y {type=MovieSceneScriptingFloatChannel, keys [[8,1],[10,1.05],[14,1]]}
         ]
        }]
      }
     ]
    }]}
 }
 ShowFlight {
  binding_count=3
  length_seconds=1.2
  object=&self:ShowFlight
  sequence {bindings [
    Fuel {
     id=#3
     possessed_class=Image
     tracks [
      {
       type=MovieSceneFloatTrack
       display_name=Render Opacity
       sections [{type=MovieSceneFloatSection, end_frame=24, start_frame=0, channels [None {type=MovieSceneScriptingFloatChannel, keys [[7,0],[17,1],[24,1]]}]}]
      }
      {
       type=MovieSceneWidgetMaterialTrack
       sections [{
         type=MovieSceneParameterSection
         end_frame=17
         start_frame=0
         channels [
          Haze {type=MovieSceneScriptingFloatChannel, keys [[7,0.17],[17,0]]}
          DelayFuelOpacity {type=MovieSceneScriptingFloatChannel, thinned_from=4, keys [[0,0],[16,0],[17,1]]}
          FuelOpacity {type=MovieSceneScriptingFloatChannel, keys [[0,0],[7,0],[8,0.25],[9,0],[10,0.5],[11,0],[17,1]]}
         ]
        }]
      }
      {
       type=MovieScene2DTransformTrack
       sections [{
         type=MovieScene2DTransformSection
         end_frame=15
         start_frame=0
         channels [
          Translation.X {type=MovieSceneScriptingFloatChannel, const [0,1]}
          Translation.Y {type=MovieSceneScriptingFloatChannel, const [0,1]}
          Angle {type=MovieSceneScriptingFloatChannel, const [0,1]}
          Scale.X {type=MovieSceneScriptingFloatChannel, const [1,1]}
          Scale.Y {type=MovieSceneScriptingFloatChannel, const [1,1]}
          Shear.X {type=MovieSceneScriptingFloatChannel, const [0,2]}
          Shear.Y {type=MovieSceneScriptingFloatChannel, thinned_from=3, keys [[8,-58.4353],[12,0]]}
         ]
        }]
      }
     ]
    }
    OutterRing {
     id=#4
     possessed_class=Image
     tracks [{
       type=MovieSceneFloatTrack
       display_name=Render Opacity
       sections [{
         type=MovieSceneFloatSection
         end_frame=17
         start_frame=0
         channels [None {type=MovieSceneScriptingFloatChannel, thinned_from=4, keys [[0,0],[9,0],[17,1]]}]
        }]
      }]
    }
    Frame {
     id=#5
     possessed_class=Image
     tracks [{
       type=MovieSceneWidgetMaterialTrack
       sections [{
         type=MovieSceneParameterSection
         end_frame=10
         start_frame=0
         channels [
          Flying? {type=MovieSceneScriptingFloatChannel, keys [[0,0],[8,1]]}
          ShowFrame? {type=MovieSceneScriptingFloatChannel, keys [[0,0],[8,1]]}
          Haze {type=MovieSceneScriptingFloatChannel, keys [[0,0],[1,0.05],[7,0]]}
         ]
        }]
      }]
    }
   ]}
 }
 Critical {
  binding_count=2
  length_seconds=0.25
  object=&self:Critical
  sequence {bindings [
    Frame {
     id=#6
     possessed_class=Image
     tracks [{
       type=MovieSceneWidgetMaterialTrack
       sections [{
         type=MovieSceneParameterSection
         channels [
          ColorR {type=MovieSceneScriptingFloatChannel, keys [[0,2],[5,1]]}
          ColorG {type=MovieSceneScriptingFloatChannel, keys [[0,0],[5,0.25]]}
          ColorB {type=MovieSceneScriptingFloatChannel, keys [[0,0],[5,0.25]]}
          ColorA {type=MovieSceneScriptingFloatChannel, const [1,2]}
         ]
        }]
      }]
    }
    Fuel {
     id=#7
     possessed_class=Image
     tracks [{
       type=MovieSceneWidgetMaterialTrack
       sections [{
         type=MovieSceneParameterSection
         end_frame=5
         start_frame=0
         channels [
          ColorR {type=MovieSceneScriptingFloatChannel, keys [[0,2],[5,1]]}
          ColorG {type=MovieSceneScriptingFloatChannel, keys [[0,0],[5,1]]}
          ColorB {type=MovieSceneScriptingFloatChannel, keys [[0,0],[5,1]]}
          ColorA {type=MovieSceneScriptingFloatChannel, const [1,2]}
          Frame ColorR {type=MovieSceneScriptingFloatChannel, const [1,2]}
          Frame ColorG {type=MovieSceneScriptingFloatChannel, keys [[0,0],[5,1]]}
          Frame ColorB {type=MovieSceneScriptingFloatChannel, keys [[0,0.008],[5,1]]}
          Frame ColorA {type=MovieSceneScriptingFloatChannel, const [1,1]}
         ]
        }]
      }]
    }
   ]}
 }
 LowFuel {
  binding_count=1
  length_seconds=0.7551
  object=&self:LowFuel
  sequence {bindings [Fuel {
     id=#8
     possessed_class=Image
     tracks [{
       type=MovieSceneWidgetMaterialTrack
       sections [{
         type=MovieSceneParameterSection
         end_frame=15
         start_frame=0
         channels [
          ColorR {type=MovieSceneScriptingFloatChannel, keys [[5,0.8294],[15,0.55]]}
          ColorG {type=MovieSceneScriptingFloatChannel, keys [[5,0.8174],[15,0]]}
          ColorB {type=MovieSceneScriptingFloatChannel, const [0,2]}
          ColorA {type=MovieSceneScriptingFloatChannel, const [1,1]}
         ]
        }]
      }]
    }]}
 }
 ShowCooldown {
  binding_count=1
  length_seconds=0.3
  object=&self:ShowCooldown
  sequence {bindings [Cooldown {
     id=#9
     possessed_class=Image
     tracks [
      {
       type=MovieSceneWidgetMaterialTrack
       sections [{type=MovieSceneParameterSection, end_frame=6, start_frame=0, channels [Haze {type=MovieSceneScriptingFloatChannel, keys [[0,0.25],[6,0]]}]}]
      }
      {
       type=MovieScene2DTransformTrack
       sections [{
         type=MovieScene2DTransformSection
         end_frame=6
         start_frame=0
         channels [
          Translation.X {type=MovieSceneScriptingFloatChannel, const [0,1]}
          Translation.Y {type=MovieSceneScriptingFloatChannel, const [0,1]}
          Angle {type=MovieSceneScriptingFloatChannel, const [0,1]}
          Scale.X {type=MovieSceneScriptingFloatChannel, const [1,1]}
          Scale.Y {type=MovieSceneScriptingFloatChannel, const [1,1]}
          Shear.X {type=MovieSceneScriptingFloatChannel, keys [[0,-78.3499],[6,0]]}
          Shear.Y {type=MovieSceneScriptingFloatChannel, const [0,1]}
         ]
        }]
      }
      {
       type=MovieSceneFloatTrack
       display_name=Render Opacity
       sections [{type=MovieSceneFloatSection, end_frame=4, start_frame=0, channels [None {type=MovieSceneScriptingFloatChannel, keys [[0,0],[4,1]]}]}]
      }
     ]
    }]}
 }
 ProgressCooldown {
  binding_count=1
  length_seconds=1
  object=&self:ProgressCooldown
  sequence {bindings [Cooldown {
     id=#10
     possessed_class=Image
     tracks [{
       type=MovieSceneWidgetMaterialTrack
       sections [{
         type=MovieSceneParameterSection
         end_frame=20
         start_frame=0
         channels [Progress {type=MovieSceneScriptingFloatChannel, keys [[0,0],[20,1]]}]
        }]
      }]
    }]}
 }
 CooldownComplete {
  binding_count=1
  length_seconds=0.2
  object=&self:CooldownComplete
  sequence {bindings [Cooldown {
     id=#11
     possessed_class=Image
     tracks [
      {
       type=MovieScene2DTransformTrack
       sections [{
         type=MovieScene2DTransformSection
         end_frame=4
         start_frame=0
         channels [
          Translation.X {type=MovieSceneScriptingFloatChannel, const [0,1]}
          Translation.Y {type=MovieSceneScriptingFloatChannel, const [0,1]}
          Angle {type=MovieSceneScriptingFloatChannel, const [0,1]}
          Scale.X {type=MovieSceneScriptingFloatChannel, keys [[0,1.25],[4,1.4]]}
          Scale.Y {type=MovieSceneScriptingFloatChannel, keys [[0,1.25],[4,1.4]]}
          Shear.X {type=MovieSceneScriptingFloatChannel, const [0,1]}
          Shear.Y {type=MovieSceneScriptingFloatChannel, const [0,1]}
         ]
        }]
      }
      {
       type=MovieSceneWidgetMaterialTrack
       sections [{type=MovieSceneParameterSection, end_frame=4, start_frame=0, channels [Haze {type=MovieSceneScriptingFloatChannel, keys [[0,0],[4,0.72]]}]}]
      }
      {
       type=MovieSceneFloatTrack
       display_name=Render Opacity
       sections [{type=MovieSceneFloatSection, end_frame=4, start_frame=0, channels [None {type=MovieSceneScriptingFloatChannel, keys [[0,1],[4,0]]}]}]
      }
     ]
    }]}
 }
 Thrust {
  binding_count=1
  length_seconds=1
  object=&self:Thrust
  sequence {
   bindings [ThrustReticle {
     id=#12
     possessed_class=Image
     tracks [
      {
       type=MovieSceneWidgetMaterialTrack
       sections [{
         type=MovieSceneParameterSection
         end_frame=30
         start_frame=0
         channels [
          TriangleSpread {type=MovieSceneScriptingFloatChannel, keys [[1,0],[29,1]]}
          RotationSwitch {type=MovieSceneScriptingFloatChannel, keys [[0,0],[20,0],[30,1]]}
          CircleOpacity {type=MovieSceneScriptingFloatChannel, keys [[0,0],[27,0],[28,1],[29,0],[30,1]]}
         ]
        }]
      }
      {
       type=MovieSceneColorTrack
       display_name=Color and Opacity
       sections [{
         type=MovieSceneColorSection
         end_frame=30
         start_frame=0
         channels [
          Color.R {type=MovieSceneScriptingFloatChannel, const [1,1]}
          Color.G {type=MovieSceneScriptingFloatChannel, const [1,1]}
          Color.B {type=MovieSceneScriptingFloatChannel, const [1,1]}
          Color.A {type=MovieSceneScriptingFloatChannel, keys [[0,0],[18,0.6018],[24,1]]}
         ]
        }]
      }
      {
       type=MovieScene2DTransformTrack
       sections [{
         type=MovieScene2DTransformSection
         end_frame=30
         start_frame=0
         channels [
          Translation.X {type=MovieSceneScriptingFloatChannel, const [0,1]}
          Translation.Y {type=MovieSceneScriptingFloatChannel, const [0,1]}
          Angle {type=MovieSceneScriptingFloatChannel, keys [[0,135],[30,0]]}
          Scale.X {type=MovieSceneScriptingFloatChannel, keys [[0,1.4],[9,1.6],[29,1]]}
          Scale.Y {type=MovieSceneScriptingFloatChannel, keys [[0,1.4],[9,1.6],[29,1]]}
          Shear.X {type=MovieSceneScriptingFloatChannel, const [0,1]}
          Shear.Y {type=MovieSceneScriptingFloatChannel, const [0,1]}
         ]
        }]
      }
     ]
    }]
   display_rate {numerator=30}
  }
 }
 ToggleUnibeam {
  binding_count=1
  length_seconds=0.4
  object=&self:ToggleUnibeam
  sequence {
   bindings [Triangle {
     id=#13
     possessed_class=Image
     tracks [{
       type=MovieSceneFloatTrack
       display_name=Render Opacity
       sections [{
         type=MovieSceneFloatSection
         end_frame=12
         start_frame=0
         channels [None {type=MovieSceneScriptingFloatChannel, keys [[0,0],[3,0],[4,0.25],[5,0],[6,0.5],[7.0005,0],[12,1]]}]
        }]
      }]
    }]
   display_rate {numerator=30}
  }
 }
 ToggleThrust {
  binding_count=1
  length_seconds=0.2667
  object=&self:ToggleThrust
  sequence {
   bindings [ThrustReticle {
     id=#14
     possessed_class=Image
     tracks [
      {
       type=MovieSceneFloatTrack
       display_name=Render Opacity
       sections [{type=MovieSceneFloatSection, end_frame=8, start_frame=0, channels [None {type=MovieSceneScriptingFloatChannel, keys [[0,0],[8,0.75]]}]}]
      }
      {
       type=MovieSceneByteTrack
       display_name=Visibility
       sections [{type=MovieSceneByteSection, end_frame=8, start_frame=0, channels [None {type=MovieSceneScriptingByteChannel, keys [[0,2],[1,0],[8,0]]}]}]
      }
     ]
    }]
   display_rate {numerator=30}
  }
 }
 ShakeFuel {
  binding_count=1
  length_seconds=1.0333
  object=&self:ShakeFuel
  sequence {
   bindings [Fuel {
     id=#15
     possessed_class=Image
     tracks [{
       type=MovieSceneWidgetMaterialTrack
       sections [{
         type=MovieSceneParameterSection
         end_frame=30
         start_frame=-1
         channels [ShakeSwitch {type=MovieSceneScriptingFloatChannel, keys [[-1,1],[30,0]]}]
        }]
      }]
    }]
   display_rate {numerator=30}
  }
 }
]
graphs [
 EventGraph {nodes [
   n1 = HandleFuel!() -> (OutputDelegate, then, New:double=~, Old:double=~) {is=title K2Node_CustomEvent}
   > Set NormalizedFuel(NormalizedFuel:double=float / float(A=@n1.New, B=100)) -> (Output_Get:double)
   > SetScalarParameterValue(
    self=Get FuelMaterial() -> (FuelMaterial:Material Instance Dynamic Object Reference)
    ParameterName=Fuel
    Value=Get NormalizedFuel() -> (NormalizedFuel:double)
   )
   > UpdateDelayedFuel()
   > SetScalarParameterValue(self=Get FuelMaterial(), ParameterName=DelayFuel, Value=Get DelayedFuel() -> (DelayedFuel:float))
   > n2 = Branch(Condition="float <= float"(A=Get NormalizedFuel()) -> (ReturnValue=~))
   > n3 = ResetAudioLowFuel()
   n4 = Delta Gate Wildcard!(
    execute:Exec=[@n3.then, @n2.else]
    Variable:"Boolean (by ref)"=Get IsCriticalState() -> (IsCriticalState:bool)
    New Value:bool="float <= float"(A=Get NormalizedFuel(), B=0.15) -> (ReturnValue=~)
   ) -> (Success:Exec, Updated Variable:"Boolean (by ref)", Previous Variable:bool=~) {is=title K2Node_MacroInstance}
   > n5 = Branch(Condition=@n4)
   n6 = Get Critical() -> (Critical:Widget Animation Object Reference)
   > n7 = "Play Animation with Finished event (legacy)"!(
    execute:Exec
    Widget:User Widget Object Reference=Self-Reference()
    InAnimation:Widget Animation Object Reference=@n6
    StartAtTime:float=0
    NumLoopsToPlay:int=0
    PlayMode:EUMGSequencePlayMode Enum
    PlaybackSpeed:float=1
   ) -> (then:Exec, Result:UMGSequence Player Object Reference, Finished:Exec) {is=title K2Node_PlayAnimation}
   > Branch(Condition="float > float"(A=Get NormalizedFuel()))
   > TriggerAudioLowFuel()
   SetVectorParameterValue(
    execute=@n7.Finished
    self=Get FrameMaterial() -> (FrameMaterial:Material Instance Dynamic Object Reference)
    ParameterName=Color
    Value=(R=0.4078,G=0.4824,B=1,A=1)
   )
   n8 = Get FuelMaterial()
   > SetVectorParameterValue(self=@n8, ParameterName=Color, Value=(R=0.8706,G=0,B=0.0078,A=1))
   > SetVectorParameterValue(self=@n8, ParameterName=Frame Color, Value=(R=0.4078,G=0.4824,B=1,A=1))
   StopAnimation(execute=@n5.else, InAnimation=@n6)
   > ResetAudioLowFuel()
   n9 = OnAimingTimerFinished() {is=title K2Node_CustomEvent}
   > PlayAnimation(InAnimation=Get ToggleThrust() -> (ToggleThrust:Widget Animation Object Reference), PlayMode=Reverse)
   > PlayAnimation(InAnimation=Get ToggleUnibeam() -> (ToggleUnibeam:Widget Animation Object Reference))
   TriggerAudioLowFuel() {is=title K2Node_CustomEvent}
   > Do Once(Reset=[@n11.else, @n12.then])
   > n10 = SpawnSound2D!(
    execute
    Sound:Sound Base Object Reference=/BarnTowelGameplay/Assets/Sounds/_various/BarnTowel_Fuel_Warning_Cue
    VolumeMultiplier:float=1
    PitchMultiplier:float=1
    StartTime:float=0
    ConcurrencySettings:Sound Concurrency Object Reference
    bPersistAcrossLevelTransition:bool=false
    bAutoDestroy:bool=true
   ) -> (then, ReturnValue:Audio Component Object Reference)
   > Set FuelWarningAudio(FuelWarningAudio:Audio Component Object Reference=@n10) -> (Output_Get:Audio Component Object Reference)
   ResetAudioLowFuel() {is=title K2Node_CustomEvent}
   > n11 = Get!(execute:Exec) -> (then:Exec, else:Exec, FuelWarningAudio:Audio Component Object Reference)
   > n12 = Stop(self:Audio Component Object Reference=@n11)
   Event Destruct()
   > Parent: Destruct()
   > n13 = Get()
   > Stop(self=@n13)
   Event Construct()
   > PlayAnimationForward(InAnimation=Get Initialize_Frame() -> (Initialize_Frame:Widget Animation Object Reference))
   > K2Node_Composite_1 = Setup Wait Gameplay Tag latent nodes!(execute) -> (then:Exec) {is=title K2Node_Composite}
   n14 = Event Tick()
   > Parent: Tick!(execute, MyGeometry:Geometry Structure=@n14.MyGeometry, InDeltaTime:float=@n14.InDeltaTime) -> (then) {is=title K2Node_CallParentFunction}
   > Update!(
    execute
    self:Fort Widget Camera Rotation Effects Object Reference=Get CameraRotationEffects() -> (CameraRotationEffects:Fort Widget Camera Rotation Effects Object Reference)
    DeltaTime:float=@n14.InDeltaTime
   ) -> (then)
   > Update!(
    execute
    self:Fort Widget Pawn Velocity Effects Object Reference=Get PawnSpeedEffects() -> (PawnSpeedEffects:Fort Widget Pawn Velocity Effects Object Reference)
    DeltaTime=@n14.InDeltaTime
    bIsFlying:bool=Get IsFlying?() -> (IsFlying?:bool)
   ) -> (then)
   Event OnInitialized()
   > n15 = Sequence!(execute) -> (then_0, then_1, then_2:Exec)
   > n16 = GetDynamicMaterial(self=Get Frame() -> (Frame:Image Object Reference))
   > Set FrameMaterial(FrameMaterial:Material Instance Dynamic Object Reference=@n16) -> (Output_Get:Material Instance Dynamic Object Reference)
   > n17 = GetDynamicMaterial(self=Get Fuel() -> (Fuel:Image Object Reference))
   > Set FuelMaterial(FuelMaterial:Material Instance Dynamic Object Reference=@n17) -> (Output_Get:Material Instance Dynamic Object Reference)
   > n18 = GetDynamicMaterial(self=Get Triangle() -> (Triangle:Image Object Reference))
   > Set TriangleMaterial(TriangleMaterial:Material Instance Dynamic Object Reference=@n18) -> (Output_Get:Material Instance Dynamic Object Reference)
   n19 = Cast To FortPlayerPawn(execute=@n15.then_1, Object=GetOwningPlayerPawn!(self:User Widget Object Reference) -> (ReturnValue:Pawn Object Reference))
   > n20 = Set Pawn(Pawn:Fort Player Pawn Object Reference=@n19) -> (Output_Get:Fort Player Pawn Object Reference)
   > n21 = WaitGameplayEventToActor!(
    execute:Exec
    TargetActor:Actor Object Reference=@n20
    EventTag:Gameplay Tag Structure=(TagName="Event.Athena.BarnTowel.StartedThrusting")
    OnlyTriggerOnce:bool=false
    OnlyMatchExact:bool=true
   ) -> (then:Exec, AsyncTaskProxy:Ability Async Wait Gameplay Event Object Reference, EventReceived:Exec, Payload:Gameplay Event Data Structure) {is=title K2Node_AsyncAction}
   > n22 = WaitGameplayEventToActor(
    TargetActor=Get Pawn() -> (Pawn:Fort Player Pawn Object Reference)
    EventTag=(TagName="Event.Athena.BarnTowel.DelayFuelUI")
    OnlyTriggerOnce=false
    OnlyMatchExact=true
   )
   > n23 = WaitGameplayEventToActor(TargetActor=Get Pawn(), EventTag=(TagName="Event.Athena.BarnTowel.StartedWindingUpThrust"), OnlyTriggerOnce=false, OnlyMatchExact=true)
   n24 = Get Thrust() -> (Thrust:Widget Animation Object Reference)
   PlayAnimation(execute=@n23.EventReceived, InAnimation=@n24, PlaybackSpeed=float / float(A=GetEndTime(self=@n24), B=Get WindUpTime() -> (WindUpTime:float)))
   > PlayAnimation(InAnimation=Get ToggleThrust())
   > PlayAnimation(InAnimation=Get ToggleUnibeam(), PlayMode=Reverse)
   > Set Timer by Event(
    Delegate=@n9
    Time=float + float(A=Get WindUpTime(), B=Get TimeThatThrustIsAllignedToCamera() -> (TimeThatThrustIsAllignedToCamera:double))
    bLooping=false
   )
   StartDelayedFuel!(execute=@n22.EventReceived, self, Duration:double=1.5) -> (then)
   > PlayAnimation(
    InAnimation=Get ShakeFuel() -> (ShakeFuel:Widget Animation Object Reference)
    PlaybackSpeed=float / float(A=1, B=Get TimeToShakeFuelBar() -> (TimeToShakeFuelBar:float))
   )
   StartDelayedFuel(execute=@n21.EventReceived, Duration=5)
   > PlayAnimation(InAnimation=Get ShakeFuel(), PlaybackSpeed=float / float(A=1, B=Get TimeToShakeFuelBar()))
   SetupWidgetEffectsObjects(execute=@n15.then_2)
   Event UpdateRemainingCooldown() {is=title K2Node_Event}
   > n25 = GetDynamicMaterial(self=Get Cooldown() -> (Cooldown:Image Object Reference))
   n26 = GetCooldownTimeRemaining!(self:Fort Weapon Reticle Extension Widget Base Object Reference) -> (RemainingTime:float, TotalDuration:float)
   > SetScalarParameterValue(self=@n25, ParameterName=Progress, Value=SafeDivide(A=@n26, B=@n26.TotalDuration))
  ]}
 Setup Wait Gameplay Tag latent nodes {
  path=EventGraph.K2Node_Composite_1.Setup Wait Gameplay Tag latent nodes
  nodes [
   Inputs()
   > n1 = Sequence!(execute) -> (then_0, then_1, then_2, then_3:Exec, then_4:Exec, then_5:Exec)
   > n2 = WaitGameplayTagAddToActor!(
    execute:Exec
    TargetActor:Actor Object Reference=Get Pawn()
    Tag:Gameplay Tag Structure=(TagName="Granted.Athena.RocketRam.State.WindingUp")
    OnlyTriggerOnce:bool=false
   ) -> (then:Exec, AsyncTaskProxy:Ability Async Wait Gameplay Tag Added Object Reference, Added:Exec) {is=title K2Node_AsyncAction}
   > n3 = WaitGameplayTagRemoveFromActor!(
    execute:Exec
    TargetActor:Actor Object Reference=Get Pawn()
    Tag:Gameplay Tag Structure=(TagName="MovementMode.Flying")
    OnlyTriggerOnce:bool=false
   ) -> (then:Exec, AsyncTaskProxy:Ability Async Wait Gameplay Tag Removed Object Reference, Removed:Exec) {is=title K2Node_AsyncAction}
   n4 = Branch(execute=@n3.Removed, Condition=Get IsFlying?())
   > PlayAnimationReverse(InAnimation=Get ShowFlight() -> (ShowFlight:Widget Animation Object Reference), PlaybackSpeed=2)
   > Set IsFlying?(IsFlying?:bool=false) -> (Output_Get:bool)
   n5 = Get ShowFlight()
   PlayAnimation(execute=@n4.else, InAnimation=@n5, StartAtTime=GetEndTime(self=@n5), PlayMode=Reverse)
   Set IsFlying?(execute=@n2.Added, IsFlying?=true)
   > PlayAnimationForward(InAnimation=Get ShowFlight())
   n6 = Branch(
    execute=@n1.then_2
    Condition=RowToBool!(
     Input:"Scalable Float Structure (by ref)"=Get SF_AllowUnibeamWithoutGauntlets() -> (SF_AllowUnibeamWithoutGauntlets:Scalable Float Structure)
     Level:double=0
    ) -> (ReturnValue:bool)
   )
   Set Visibility(execute=[@n9.else, @n6.then], Visibility=HitTestInvisible, self=Get Triangle())
   > PlayAnimationForward(InAnimation=Get Initialize_Cannon() -> (Initialize_Cannon:Widget Animation Object Reference))
   n7 = WaitGameplayTagAddToActor(execute=@n6.else, TargetActor=Get Pawn(), Tag=(TagName="Granted.HighTower.Power.Tomato.Combo.InInventory"), OnlyTriggerOnce=false)
   > n8 = WaitGameplayTagRemoveFromActor(TargetActor=Get Pawn(), Tag=(TagName="Granted.HighTower.Power.Tomato.Combo.InInventory"), OnlyTriggerOnce=false)
   Set Visibility(execute=@n8.Removed, Visibility=Hidden, self=Get Triangle())
   n9 = Branch(
    execute=@n7.Added
    Condition=RowToBool(Input=Get SF_FullyDisableUnibeamWhileFlying() -> (SF_FullyDisableUnibeamWhileFlying:Scalable Float Structure), Level=0)
   )
   n10 = WaitGameplayTagAddToActor(execute=@n1.then_3, TargetActor=Get Pawn(), Tag=(TagName="Granted.Cooldown.HighTower.Tomato.RepulsorCannon"), OnlyTriggerOnce=false)
   > n11 = WaitGameplayTagRemoveFromActor(TargetActor=Get Pawn(), Tag=(TagName="Granted.Cooldown.HighTower.Tomato.RepulsorCannon"), OnlyTriggerOnce=false)
   n12 = Get TriangleMaterial() -> (TriangleMaterial:Material Instance Dynamic Object Reference)
   SetScalarParameterValue(execute=@n11.Removed, self=@n12, ParameterName=CooldownProgress, Value=0)
   > SetScalarParameterValue(self=@n12, ParameterName=Ready?, Value=1)
   > Set RenderOpacity!(execute, RenderOpacity:float=1, self:Widget Object Reference=Get Triangle()) -> (then, Output_Get:float)
   > Clear and Invalidate Timer by Handle(Handle=Get GauntletCooldownQueryTimer() -> (GauntletCooldownQueryTimer:Timer Handle Structure))
   SetScalarParameterValue(execute=@n10.Added, self=Get TriangleMaterial(), ParameterName=Ready?, Value=0)
   > Set RenderOpacity(RenderOpacity=0.1, self=Get Triangle())
   > StartGauntletCooldownQuery()
   Outputs!(then:Exec=@n1.then_5)
   n13 = WaitForAttributeChanged!(
    execute:Exec=@n1.then_1
    TargetActor:Actor Object Reference=GetOwningPlayer()
    Attribute:Gameplay Attribute Structure=(AttributeName="LocalFuel",Attribute=/Script/FlyingCodeRuntime.FlyingAttributeSet:LocalFuel,AttributeOwner="&Class'&FlyingAttributeSet'")
    OnlyTriggerOnce:bool=false
   ) -> (
    then:Exec
    AsyncTaskProxy:Ability Async Wait Attribute Changed Object Reference
    Changed:Exec
    Attribute:Gameplay Attribute Structure
    NewValue:float=~
    OldValue:float=~
   ) {is=title K2Node_AsyncAction}
   HandleFuel!(execute=@n13.Changed, self, New:double=@n13.NewValue, Old:double=@n13.OldValue) -> (then) {is=title K2Node_CallFunction}
   Set InverseCoolDownTime(execute=@n1.then_4, InverseCoolDownTime:double=float / float(A=1, B=Get Cooldown time() -> (Cooldown time:double))) -> (Output_Get:double)
   > n14 = WaitGameplayTagAddToActor(TargetActor=Get Pawn(), Tag=(TagName="Granted.Cooldown.BarnTowel"), OnlyTriggerOnce=false)
   > n15 = WaitGameplayTagRemoveFromActor(TargetActor=Get Pawn(), Tag=(TagName="Granted.Cooldown.BarnTowel"), OnlyTriggerOnce=false)
   HandleCooldownTimer!(execute=@n15.Removed, self:Fort Weapon Reticle Extension Widget Base Object Reference, bActivateTimer:bool=false, TimerInterval:float=0) -> (then)
   > Branch(Condition=Get IsCoolingDown?() -> (IsCoolingDown?:bool))
   > PlayAnimation(InAnimation=Get CooldownComplete() -> (CooldownComplete:Widget Animation Object Reference))
   > Set IsCoolingDown?(IsCoolingDown?:bool=false) -> (Output_Get:bool)
   HandleCooldownTimer(execute=@n14.Added, bActivateTimer=true, TimerInterval=0.1)
   > PlayAnimation(InAnimation=Get ShowCooldown() -> (ShowCooldown:Widget Animation Object Reference))
   > Set IsCoolingDown?(IsCoolingDown?=true)
  ]
 }
 GauntletCooldownHandler {nodes [
   GauntletCooldownHandler() {is=title K2Node_FunctionEntry}
   > n1 = Get Active Gameplay Effects for Query!(
    execute
    self:Ability System Component Object Reference=GetAbilitySystemComponent!(Actor:Actor Object Reference=Get Pawn()) -> (ReturnValue:Ability System Component Object Reference)
    Query:"Gameplay Effect Query Structure (by ref)"=Make Gameplay Effect Query!(
     CustomMatchDelegate_BP:Delegate="(null).None"
     OwningTagQuery:Gameplay Tag Query Structure=(TokenStreamVersion=0,TagDictionary=((TagName="Granted.Cooldown.HighTower.Tomato.RepulsorCannon")),QueryTokenStream=(0,1,2,1,0),UserDescription="",AutoDescription=" ALL( Granted.Cooldown.HighTower.Tomato.RepulsorCannon )")
     EffectTagQuery:Gameplay Tag Query Structure=(TokenStreamVersion=0,TagDictionary=,QueryTokenStream=,UserDescription="",AutoDescription="")
     SourceTagQuery:Gameplay Tag Query Structure=(TokenStreamVersion=0,TagDictionary=,QueryTokenStream=,UserDescription="",AutoDescription="")
     SourceAggregateTagQuery:Gameplay Tag Query Structure=(TokenStreamVersion=0,TagDictionary=,QueryTokenStream=,UserDescription="",AutoDescription="")
     ModifyingAttribute:Gameplay Attribute Structure=(AttributeName="",Attribute=,AttributeOwner=None)
     EffectSource:Object Reference
     EffectDefinition:Gameplay Effect Class Reference
    ) -> (GameplayEffectQuery:Gameplay Effect Query Structure) {is=title K2Node_MakeStruct}
   ) -> (then, ReturnValue:Array of Active Gameplay Effect Handle Structures) {is=title K2Node_CallFunction}
   > Branch(Condition="integer > integer"(A=Length(TargetArray:Array of Active Gameplay Effect Handle Structures=@n1)))
   n2 = "Get (a ref)"(Array:Array of Active Gameplay Effect Handle Structures=@n1, Dimension 1=0) -> (Output:"Active Gameplay Effect Handle Structure (by ref)")
   > n3 = GetActiveGameplayEffectRemainingDuration!(execute, ActiveHandle:Active Gameplay Effect Handle Structure=@n2) -> (then, ReturnValue:float)
   > n4 = GetActiveGameplayEffectTotalDuration!(execute, ActiveHandle:Active Gameplay Effect Handle Structure=@n2) -> (then, ReturnValue:float)
   > SetScalarParameterValue(self=Get TriangleMaterial(), ParameterName=CooldownProgress, Value=float - float(A=1, B=float / float(A=@n3, B=@n4)))
  ]}
 SetupWidgetEffectsObjects {nodes [
   SetupWidgetEffectsObjects() {is=title K2Node_FunctionEntry}
   > n1 = CreatePawnVelocityEffectsObject!(execute, Outer:Object Reference) -> (then, ReturnValue:Fort Widget Pawn Velocity Effects Object Reference)
   > n2 = Set PawnSpeedEffects(PawnSpeedEffects:Fort Widget Pawn Velocity Effects Object Reference=@n1) -> (Output_Get:Fort Widget Pawn Velocity Effects Object Reference)
   > Setup!(
    execute
    self:Fort Widget Pawn Velocity Effects Object Reference=@n2
    InPawn:Pawn Object Reference=Get Pawn()
    InSpeedEffectData:"Basic Effect Data Structure (by ref)"=Make Basic Effect Data!(Material:Material Instance Dynamic Object Reference=Get FrameMaterial(), ParameterName:Name=Speed) -> (BasicEffectData:Basic Effect Data Structure) {is=title K2Node_MakeStruct}
    InOverdriveEffectData:"Basic Effect Data Structure (by ref)"=Make Basic Effect Data(Material=Get FrameMaterial(), ParameterName=Overdrive?)
    InMaxSpeed:float=2500
   ) -> (then)
   > n3 = Create UICamera Effects from Image List!(
    execute:Exec
    ImageList:Array of Image Object References=Make Array!(
     "[0]":Image Object Reference=Get Frame()
     "[1]":Image Object Reference=Get Fuel()
     "[2]":Image Object Reference=Get OutterRing() -> (OutterRing:Image Object Reference)
    ) -> (Array:Array of Image Object References)
    PitchMaterial:Material Instance Dynamic Object Reference=Get FrameMaterial()
   ) -> (then:Exec, CameraEffectsObject:"Fort Widget Camera Rotation Effects Object Reference (by ref)") {is=title K2Node_MacroInstance}
   > Set CameraRotationEffects(CameraRotationEffects:Fort Widget Camera Rotation Effects Object Reference=@n3) -> (Output_Get:Fort Widget Camera Rotation Effects Object Reference)
  ]}
 StartDelayedFuel {nodes [
   n1 = StartDelayedFuel!() -> (then, Duration:double=~) {is=title K2Node_FunctionEntry}
   n2 = GetGameTimeInSeconds()
   > Set FuelDelayedUntilTime(FuelDelayedUntilTime:double="Max (Float)"(A=float + float(A=@n1, B=@n2), B=Get FuelDelayedUntilTime() -> (FuelDelayedUntilTime:double))) -> (Output_Get:double)
  ]}
 StartGauntletCooldownQuery {nodes [
   StartGauntletCooldownQuery() {is=title K2Node_FunctionEntry}
   > n1 = Set Timer by Event(Delegate=Create Event(), Time=0.1, bLooping=true, bMaxOncePerFrame=true)
   > Set GauntletCooldownQueryTimer(GauntletCooldownQueryTimer:Timer Handle Structure=@n1) -> (Output_Get:Timer Handle Structure)
   > GauntletCooldownHandler() {is=title K2Node_CallFunction}
  ]}
 UpdateDelayedFuel {nodes [
   UpdateDelayedFuel() {is=title K2Node_FunctionEntry}
   > n1 = Set TimeRemaining(TimeRemaining:double=float - float(A=Get FuelDelayedUntilTime(), B=GetGameTimeInSeconds())) -> (Output_Get:double)
   > n2 = Branch(Condition="float <= float"(A=@n1) -> (ReturnValue=~))
   n3 = float - float(A=GetGameTimeInSeconds(), B=Get LastDelayedFuelUpdateTime() -> (LastDelayedFuelUpdateTime:double))
   > n4 = Set DelayedFuel(DelayedFuel:float=FInterpTo_Constant!(Current:double=Get DelayedFuel(), Target:double=Get NormalizedFuel(), DeltaTime:double=@n3, InterpSpeed:double=0.3) -> (ReturnValue:double)) -> (Output_Get:float)
   Set LastDelayedFuelUpdateTime(execute=[@n4.then, @n5.then], LastDelayedFuelUpdateTime:double=GetGameTimeInSeconds()) -> (Output_Get:double)
   n5 = Set DelayedFuel(execute=@n2.else, DelayedFuel=FInterpTo_Constant(Current=Get DelayedFuel(), Target=Get NormalizedFuel(), DeltaTime=@n3, InterpSpeed=0.03))
  ]}
]
variables [
 CameraRotationEffects {type=FortWidgetCameraRotationEffects, default=null}
 Cooldown time {type=number, default=20}
 DelayedFuel {type=number, default=1}
 FrameMaterial {type=MaterialInstanceDynamic}
 FuelDelayedUntilTime {type=number}
 FuelMaterial {type=MaterialInstanceDynamic}
 FuelWarningAudio {type=AudioComponent, default=null}
 GauntletCooldownQueryTimer {type=TimerHandle}
 InverseCoolDownTime {type=number}
 IsCoolingDown? {type=boolean}
 IsCriticalState {type=boolean}
 IsFlying? {type=boolean}
 LastDelayedFuelUpdateTime {type=number}
 NormalizedFuel {type=number, default=1}
 Pawn {type=FortPlayerPawn}
 PawnSpeedEffects {type=FortWidgetPawnVelocityEffects, default=null}
 SF_AllowUnibeamWithoutGauntlets {
  type=ScalableFloat
  default {
   registryType=None
   value=1
   curve {curveTable=/BarnTowelGameplay/DataTables/BarnTowelGameplayGameData, rowName=Default.BarnTowelGameplay.Unibeam.AllowUnibeamWithoutGauntlets}
  }
 }
 SF_FullyDisableUnibeamWhileFlying {
  type=ScalableFloat
  default {
   registryType=None
   value=1
   curve {curveTable=&BarnTowelGameplayGameData, rowName=Default.BarnTowelGameplay.Unibeam.FullyDisableUnibeamWhileFlying}
  }
 }
 TimeThatThrustIsAllignedToCamera {type=number, default=3}
 TimeToShakeFuelBar {type=number, default=0.6}
 TriangleMaterial {type=MaterialInstanceDynamic}
 WindUpTime {type=number, default=0.7}
]
widgets {
 info {
  inheritedWidgetCount=1
  parentClass=/Script/FortniteUI.FortWeaponReticleExtensionWidgetBase
  rootWidgetClass=&Overlay
  widgetCount=9
 }
 widgets [
  Overlay_0 {widgetClassPath=&Overlay, widget=Overlay_0}
  Triangle {
   widgetClassPath=&Image
   bIsVariable=true
   parent=Overlay_0
   widget=Triangle
   properties {
    visibility=Hidden
    brush {
     bIsDynamicallyLoaded=false
     drawAs=Image
     imageType=FullColor
     mirroring=NoMirror
     resourceName=None
     resourceObject=/BarnTowelGameplay/UI/M_UI_Barntowel_Cannon
     tiling=NoTile
     imageSize=<256 256>
     margin=<ltrb 0 0 0 0>
     outlineSettings {
      bUseBrushTransparency=false
      roundingType=HalfHeightRadius
      width=0
      color {colorUseRule=UseColor_Specified, specifiedColor=<rgba 0 0 0 0>}
      cornerRadii=<0 0 0 0>
     }
     tintColor {colorUseRule=UseColor_Specified, specifiedColor=<rgba 1 1 1 1>}
     uVRegion {bIsValid=false, max=<0 0>, min=<0 0>}
    }
    renderTransform {angle=0, scale=<1 1>, shear=<0 0>, translation=<0 -5>}
   }
   slot {class=OverlaySlot, properties {horizontalAlignment=HAlign_Center, verticalAlignment=VAlign_Center, padding=<ltrb 0 5 0 0>}}
  }
  Frame {
   widgetClassPath=&Image
   bIsVariable=true
   parent=Overlay_0
   widget=Frame
   properties {
    visibility=HitTestInvisible
    brush {
     bIsDynamicallyLoaded=false
     drawAs=Image
     imageType=FullColor
     mirroring=NoMirror
     resourceName=None
     resourceObject=&/M_UI_Barntowel_V2
     tiling=NoTile
     imageSize=<1920 1080>
     margin=<ltrb 0 0 0 0>
     outlineSettings {
      bUseBrushTransparency=false
      roundingType=HalfHeightRadius
      width=0
      color {colorUseRule=UseColor_Specified, specifiedColor=<rgba 0 0 0 0>}
      cornerRadii=<0 0 0 0>
     }
     tintColor {colorUseRule=UseColor_Specified, specifiedColor=<rgba 1 1 1 1>}
     uVRegion {bIsValid=false, max=<0 0>, min=<0 0>}
    }
   }
   slot {class=OverlaySlot, properties {horizontalAlignment=HAlign_Center, verticalAlignment=VAlign_Center}}
  }
  OutterRing {
   widgetClassPath=&Image
   bIsVariable=true
   parent=Overlay_0
   widget=OutterRing
   properties {
    renderOpacity=0
    visibility=HitTestInvisible
    brush {
     bIsDynamicallyLoaded=false
     drawAs=Image
     imageType=FullColor
     mirroring=NoMirror
     resourceName=None
     resourceObject=/BarnTowelGameplay/UI/M_UI_BarnTowelOutterRing
     tiling=NoTile
     imageSize=<1080 1080>
     margin=<ltrb 0 0 0 0>
     outlineSettings {
      bUseBrushTransparency=false
      roundingType=HalfHeightRadius
      width=0
      color {colorUseRule=UseColor_Specified, specifiedColor=<rgba 0 0 0 0>}
      cornerRadii=<0 0 0 0>
     }
     tintColor {colorUseRule=UseColor_Specified, specifiedColor=<rgba 1 1 1 1>}
     uVRegion {bIsValid=false, max=<0 0>, min=<0 0>}
    }
    renderTransform {angle=0, scale=<0.94 0.94>, shear=<0 0>, translation=<0 0>}
   }
   slot {class=OverlaySlot, properties {horizontalAlignment=HAlign_Center, verticalAlignment=VAlign_Center}}
  }
  Fuel {
   widgetClassPath=&Image
   bIsVariable=true
   parent=Overlay_0
   widget=Fuel
   properties {
    visibility=HitTestInvisible
    brush {
     bIsDynamicallyLoaded=false
     drawAs=Image
     imageType=FullColor
     mirroring=NoMirror
     resourceName=None
     resourceObject=&/MI_UI_Barntowel_Fuel
     tiling=NoTile
     imageSize=<91.4286 731.4286>
     margin=<ltrb 0 0 0 0>
     outlineSettings {
      bUseBrushTransparency=false
      roundingType=HalfHeightRadius
      width=0
      color {colorUseRule=UseColor_Specified, specifiedColor=<rgba 0 0 0 0>}
      cornerRadii=<0 0 0 0>
     }
     tintColor {colorUseRule=UseColor_Specified, specifiedColor=<rgba 1 1 1 1>}
     uVRegion {bIsValid=false, max=<0 0>, min=<0 0>}
    }
   }
   slot {class=OverlaySlot, properties {horizontalAlignment=HAlign_Center, verticalAlignment=VAlign_Center, padding=<ltrb 360 0 0 0>}}
  }
  CenterDot {
   widgetClassPath=&Image
   bIsVariable=true
   parent=Overlay_0
   widget=CenterDot
   properties {
    visibility=HitTestInvisible
    brush {
     bIsDynamicallyLoaded=false
     drawAs=Image
     imageType=FullColor
     mirroring=NoMirror
     resourceName=None
     resourceObject=/Game/Athena/HUD/reticles/reticle_dot_new
     tiling=NoTile
     imageSize=<8 8>
     margin=<ltrb 0 0 0 0>
     outlineSettings {
      bUseBrushTransparency=false
      roundingType=HalfHeightRadius
      width=0
      color {colorUseRule=UseColor_Specified, specifiedColor=<rgba 0 0 0 0>}
      cornerRadii=<0 0 0 0>
     }
     tintColor {colorUseRule=UseColor_Specified, specifiedColor=<rgba 1 1 1 1>}
     uVRegion {bIsValid=false, max=<0 0>, min=<0 0>}
    }
   }
   slot {class=OverlaySlot, properties {horizontalAlignment=HAlign_Center, verticalAlignment=VAlign_Center}}
  }
  ThrustReticle {
   widgetClassPath=&Image
   bIsVariable=true
   parent=Overlay_0
   widget=ThrustReticle
   properties {
    visibility=Hidden
    brush {
     bIsDynamicallyLoaded=false
     drawAs=Image
     imageType=FullColor
     mirroring=NoMirror
     resourceName=None
     resourceObject=&/PROTO_M_ThrustingReticle
     tiling=NoTile
     imageSize=<100 100>
     margin=<ltrb 0 0 0 0>
     outlineSettings {
      bUseBrushTransparency=false
      roundingType=HalfHeightRadius
      width=0
      color {colorUseRule=UseColor_Specified, specifiedColor=<rgba 0 0 0 0>}
      cornerRadii=<0 0 0 0>
     }
     tintColor {colorUseRule=UseColor_Specified, specifiedColor=<rgba 1 1 1 1>}
     uVRegion {bIsValid=false, max=<0 0>, min=<0 0>}
    }
   }
   slot {class=OverlaySlot, properties {horizontalAlignment=HAlign_Center, verticalAlignment=VAlign_Center}}
  }
  Cooldown {
   widgetClassPath=&Image
   bIsVariable=true
   parent=Overlay_0
   widget=Cooldown
   properties {
    renderOpacity=0
    visibility=HitTestInvisible
    brush {
     bIsDynamicallyLoaded=false
     drawAs=Image
     imageType=FullColor
     mirroring=NoMirror
     resourceName=None
     resourceObject=&/M_UI_Barntowel_Cooldown
     tiling=NoTile
     imageSize=<128 16>
     margin=<ltrb 0 0 0 0>
     outlineSettings {
      bUseBrushTransparency=false
      roundingType=HalfHeightRadius
      width=0
      color {colorUseRule=UseColor_Specified, specifiedColor=<rgba 0 0 0 0>}
      cornerRadii=<0 0 0 0>
     }
     tintColor {colorUseRule=UseColor_Specified, specifiedColor=<rgba 1 1 1 1>}
     uVRegion {bIsValid=false, max=<0 0>, min=<0 0>}
    }
   }
   slot {class=OverlaySlot, properties {horizontalAlignment=HAlign_Center, verticalAlignment=VAlign_Center, padding=<ltrb 0 0 0 75>}}
  }
  TutorialHighlight {widgetClassPath=&TutorialHighlight}
 ]
}
